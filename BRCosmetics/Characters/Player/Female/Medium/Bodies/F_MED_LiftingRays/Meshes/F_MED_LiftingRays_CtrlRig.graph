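# Control Rig for the LiftingRays female medium body that corrects bandolier, purse, shorts and leg deformation as the character moves.
# It runs shoulder-volume and shared DeformRig functions, then weighted multi-bone parent constraints that branch on rigid-body/grind-rail and skydiving state, then crouch/skydive fixes.
# Relative bone rotations, spherical pose readers and interpolated gameplay flags weight additive bone offsets.
# A right-arm pose reader also lifts the bandolier root and sets its constraint alpha.
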ControlRigBlueprint /BRCosmetics/Characters/Player/Female/Medium/Bodies/F_MED_LiftingRays/Meshes/F_MED_LiftingRays_CtrlRig
graphs [
 New Function_ContainedGraph {nodes [
   n1 = RigVMFunctionEntryNode()
   n2 = &RigVMFunction_ControlFlowBranch(ExecuteContext=@n1.ExecuteContext, Condition=RigVMVariableNode() {variable=RBEnableAndGrindRail}.Value)
   RigVMFunctionReturnNode(ExecuteContext=@n2.Completed)
   SetupNORban = RigVMFunctionReferenceNode(ExecuteContext=@n2.False)
   RigVMFunction_ControlFlowBranch = &RigVMFunction_ControlFlowBranch(ExecuteContext=@n2.True, Condition=RigVMVariableNode() {variable=Skydiving}.Value)
   &RigUnit_ParentConstraint(
    AdvancedSettings=(RotationOrderForFilter=XZY)
    Child=(Type=Bone,Name="base_bandolier_root_skin")
    Filter=(TranslationFilter=(bX=True,bY=True,bZ=True),RotationFilter=(bX=False,bY=False,bZ=False),ScaleFilter=(bX=True,bY=True,bZ=True))
    Parents="((Item=(Type=Bone,Name=\"deform_upperarm_fr_r\"),Weight=0.0014),(Item=(Type=Bone,Name=\"deform_upperarm_up_r\"),Weight=0.015),(Item=(Type=Bone,Name=\"upperarm_twist_01_r\"),Weight=0.0159),(Item=(Type=Bone,Name=\"spine_05\"),Weight=0.1643),(Item=(Type=Bone,Name=\"cl ...(truncated)"
    Weight=1
    ExecutePin=@RigVMFunction_ControlFlowBranch.False
   )
   > &RigUnit_ParentConstraint(
    AdvancedSettings=(RotationOrderForFilter=XZY)
    Child=(Type=Bone,Name="dyn_bandolier_fr_01_l")
    Parents="((Item=(Type=Bone,Name=\"deform_upperarm_fr_r\"),Weight=0.0152),(Item=(Type=Bone,Name=\"upperarm_twist_01_r\"),Weight=0.0186),(Item=(Type=Bone,Name=\"deform_pec_r\"),Weight=0.0324),(Item=(Type=Bone,Name=\"spine_04\"),Weight=0.1139),(Item=(Type=Bone,Name=\"clavicle_ ...(truncated)"
    Weight=1
   )
   > &RigUnit_ParentConstraint(
    AdvancedSettings=(RotationOrderForFilter=XZY)
    Child=(Type=Bone,Name="dyn_bandolier_fr_02_l")
    Filter=(TranslationFilter=(bX=True,bY=True,bZ=True),RotationFilter=(bX=True,bY=True,bZ=False),ScaleFilter=(bX=True,bY=True,bZ=True))
    Parents=((Item=(Type=Bone,Name="clavicle_r"),Weight=0.0236),(Item=(Type=Bone,Name="deform_pec_r"),Weight=0.1415),(Item=(Type=Bone,Name="spine_04"),Weight=0.3141),(Item=(Type=Bone,Name="spine_05"),Weight=0.5208))
    Weight=1
   )
   n3 = &RigVMFunction_AlphaInterp(
    InterpSpeedDecreasing=10
    InterpSpeedIncreasing=10
    Scale=1
    bClampResult=true
    Value=RigVMVariableNode() {variable=TacticalSprintingJump}.Value
   )
   n4 = &RigVMFunction_MathDoubleRemap(SourceMaximum=1, SourceMinimum=0, TargetMaximum=0, TargetMinimum=1, Value=@n3.Result)
   > &RigUnit_ParentConstraint(
    AdvancedSettings=(RotationOrderForFilter=XZY)
    Child=(Type=Bone,Name="base_bandolier_root_skin")
    Parents="((Item=(Type=Bone,Name=\"deform_upperarm_fr_r\"),Weight=0.0011),(Item=(Type=Bone,Name=\"deform_upperarm_up_r\"),Weight=0.0118),(Item=(Type=Bone,Name=\"upperarm_twist_01_r\"),Weight=0.0125),(Item=(Type=Bone,Name=\"spine_04\"),Weight=0.0176),(Item=(Type=Bone,Name=\"s ...(truncated)"
    Weight=@n4.Result
   )
   > &RigUnit_ParentConstraint(
    AdvancedSettings=(RotationOrderForFilter=XZY)
    Child=(Type=Bone,Name="dyn_bandolier_bk_01_extra_l")
    Filter=(TranslationFilter=(bX=True,bY=True,bZ=True),RotationFilter=(bX=False,bY=False,bZ=False),ScaleFilter=(bX=True,bY=True,bZ=True))
    Parents="((Item=(Type=Bone,Name=\"clavicle_l\"),Weight=0.0012),(Item=(Type=Bone,Name=\"spine_04\"),Weight=0.0163),(Item=(Type=Bone,Name=\"deform_upperarm_up_r\"),Weight=0.0263),(Item=(Type=Bone,Name=\"upperarm_twist_01_r\"),Weight=0.0388),(Item=(Type=Bone,Name=\"deform_uppe ...(truncated)"
    Weight=@n4.Result
   )
   n5 = RigVMVariableNode() {variable=ArmBackInMelee}
   Remap = &RigVMFunction_MathDoubleRemap(SourceMaximum=1, SourceMinimum=0, TargetMaximum=0.5, TargetMinimum=0, bClamp=false, Value=@n5.Value)
   AlphaInterp = &RigVMFunction_AlphaInterp(InterpSpeedDecreasing=20, InterpSpeedIncreasing=20, Scale=1, bClampResult=true, Value=@Remap.Result)
   > &RigUnit_ParentConstraint(
    AdvancedSettings=(InterpolationType=Average,RotationOrderForFilter=XZY)
    Child=(Type=Bone,Name="base_purse_l")
    Parents=((Item=(Type=Bone,Name="pelvis"),Weight=1))
    Weight=@AlphaInterp.Result
   )
   > Delta Rotation Z = RigVMFunctionReferenceNode(Child=head, Parent=spine_05, SourceMax=25, SourceMin=15)
   Multiply = &RigVMFunction_MathDoubleMul(A=@"Delta Rotation Z".Delta Z, B=@n5.Value)
   > &RigUnit_ModifyTransforms(
    ItemToModify="((Item=(Type=Bone,Name=\"dyn_purse_03_l\"),Transform=(Rotation=(X=0,Y=-0.2455,Z=0,W=0.9694),Translation=(X=-3.0157,Y=0,Z=0),Scale3D=(X=1,Y=1,Z=1))),(Item=(Type=Bone,Name=\"dyn_purse_01_l\"),Transform=(Rotation=(X=0,Y=0,Z=0,W=1),Translation=(X=-5,Y=2,Z=3),Scale ...(truncated)"
    WeightMinimum=0
    Weight=@Multiply.Result
   )
   > DeltaRotation Y = RigVMFunctionReferenceNode(Child=clavicle_r, Parent=spine_05, Source Max=50, Source Min=70)
   > &RigUnit_ModifyTransforms(
    ItemToModify="((Item=(Type=Bone,Name=\"dyn_bandolier_fr_01_extra_l\"),Transform=(Rotation=(X=0,Y=0.144,Z=0,W=-0.9896),Translation=(X=0.7253,Y=0.4696,Z=0.5909),Scale3D=(X=1,Y=1,Z=1))),(Item=(Type=Bone,Name=\"dyn_bandolier_fr_02_l\"),Transform=(Rotation=(X=0,Y=0,Z=0.0196,W=0. ...(truncated)"
    WeightMinimum=0
    Weight=@"DeltaRotation Y".Delta Y
   )
   > Delta Rotation Z_3 = RigVMFunctionReferenceNode(Child=lowerarm_r, Parent=upperarm_r, SourceMax=35, SourceMin=60)
   > &RigUnit_ModifyTransforms(
    ItemToModify="((Item=(Type=Bone,Name=\"dyn_bandolier_fr_02_extra_l\"),Transform=(Rotation=(X=0,Y=0,Z=0,W=-1),Translation=(X=0,Y=0.3493,Z=0),Scale3D=(X=1,Y=1,Z=1))),(Item=(Type=Bone,Name=\"dyn_bandolier_fr_01_extra_l\"),Transform=(Rotation=(X=0,Y=0,Z=0,W=-1),Translation=(X=0 ...(truncated)"
    WeightMinimum=0
    Weight=@"Delta Rotation Z_3".Delta Z
   )
   > Delta Rotation Z_2 = RigVMFunctionReferenceNode(Child=spine_05, Parent=spine_04, SourceMax=10, SourceMin=2)
   > &RigUnit_ModifyTransforms(
    ItemToModify=((Item=(Type=Bone,Name="dyn_bandolier_fr_02_l"),Transform=(Rotation=(X=0,Y=0,Z=0.0194,W=0.9998),Translation=(X=-0.0407,Y=2.3792,Z=0),Scale3D=(X=1,Y=1,Z=1))))
    WeightMinimum=0
    Weight=@"Delta Rotation Z_2".Delta Z
   )
   > SetupRail = RigVMFunctionReferenceNode()
   &RigUnit_ParentConstraint(
    AdvancedSettings=(RotationOrderForFilter=XZY)
    Child=(Type=Bone,Name="base_bandolier_root")
    Parents="((Item=(Type=Bone,Name=\"deform_upperarm_fr_r\"),Weight=0.0011),(Item=(Type=Bone,Name=\"deform_upperarm_up_r\"),Weight=0.0118),(Item=(Type=Bone,Name=\"upperarm_twist_01_r\"),Weight=0.0125),(Item=(Type=Bone,Name=\"spine_04\"),Weight=0.0176),(Item=(Type=Bone,Name=\"s ...(truncated)"
    ExecutePin=@RigVMFunction_ControlFlowBranch.True
    Weight=RigVMVariableNode() {variable=bandolier constraint alpha}.Value
   )
   > &RigUnit_ParentConstraint(
    AdvancedSettings=(RotationOrderForFilter=XZY)
    Child=(Type=Bone,Name="base_bandolier_bk_01_l")
    Parents="((Item=(Type=Bone,Name=\"clavicle_l\"),Weight=0.0012),(Item=(Type=Bone,Name=\"spine_04\"),Weight=0.0163),(Item=(Type=Bone,Name=\"deform_upperarm_up_r\"),Weight=0.0264),(Item=(Type=Bone,Name=\"upperarm_twist_01_r\"),Weight=0.0389),(Item=(Type=Bone,Name=\"deform_uppe ...(truncated)"
    Weight=1
   )
   > &RigUnit_ParentConstraint(
    AdvancedSettings=(RotationOrderForFilter=XZY)
    Child=(Type=Bone,Name="dyn_bandolier_fr_01_l")
    Parents="((Item=(Type=Bone,Name=\"deform_upperarm_fr_r\"),Weight=0.0152),(Item=(Type=Bone,Name=\"upperarm_twist_01_r\"),Weight=0.0186),(Item=(Type=Bone,Name=\"deform_pec_r\"),Weight=0.0324),(Item=(Type=Bone,Name=\"spine_04\"),Weight=0.1139),(Item=(Type=Bone,Name=\"clavicle_ ...(truncated)"
    Weight=1
   )
   > &RigUnit_ParentConstraint(
    AdvancedSettings=(RotationOrderForFilter=XZY)
    Child=(Type=Bone,Name="dyn_bandolier_fr_02_l")
    Parents=((Item=(Type=Bone,Name="clavicle_r"),Weight=0.0217),(Item=(Type=Bone,Name="deform_pec_r"),Weight=0.1459),(Item=(Type=Bone,Name="spine_04"),Weight=0.3157),(Item=(Type=Bone,Name="spine_05"),Weight=0.5166))
    Weight=1
   )
  ]}
 New Function_ContainedGraph {nodes [
   RigVMFunctionEntryNode()
   > RigVMFunction_Sequence = RigVMAggregateNode()
   n1 = &RigVMFunction_AlphaInterp(
    InterpSpeedDecreasing=8
    InterpSpeedIncreasing=8
    Scale=1
    bClampResult=true
    Value=&RigVMFunction_MathBoolToFloat(Value=RigVMVariableNode() {variable=Crouching}.Value).Result
   )
   &RigUnit_ModifyTransforms(
    ItemToModify="((Item=(Type=Bone,Name=\"base_short_fr_l\"),Transform=(Rotation=(X=-0.1231,Y=-0.0649,Z=0.0081,W=-0.9902),Translation=(X=-0.1311,Y=0,Z=-0.9959),Scale3D=(X=1,Y=1,Z=1))),(Item=(Type=Bone,Name=\"base_short_sd_l\"),Transform=(Rotation=(X=0.0247,Y=-0.0269,Z=0.0427,W ...(truncated)"
    WeightMinimum=0
    Weight=@n1.Result
    ExecutePin=@RigVMFunction_Sequence.A
   )
   RigVMFunctionReturnNode(ExecuteContext=@RigVMFunction_Sequence.C)
   n2 = &RigVMFunction_AlphaInterp(
    InterpSpeedDecreasing=7
    InterpSpeedIncreasing=8
    Scale=1
    bClampResult=true
    Value=&RigVMFunction_MathBoolToFloat(Value=RigVMVariableNode() {variable=SkyDiveMode}.Value).Result
   )
   &RigUnit_ModifyTransforms(
    ItemToModify="((Item=(Type=Bone,Name=\"base_purse_l\"),Transform=(Rotation=(X=0,Y=0.1736,Z=0,W=0.9848),Translation=(X=0,Y=0,Z=0),Scale3D=(X=1,Y=1,Z=1))),(Item=(Type=Bone,Name=\"dyn_purse_02_l\"),Transform=(Rotation=(X=0,Y=0.342,Z=0,W=0.9397),Translation=(X=0,Y=0,Z=0),Scale3 ...(truncated)"
    WeightMinimum=0
    Weight=@n2.Result
    ExecutePin=@RigVMFunction_Sequence.B
   )
   RigVMFunction_MathBoolToFloat = &RigVMFunction_MathBoolToFloat(Value=RigVMVariableNode() {variable=Skydiving}.Value)
   n3 = &RigVMFunction_AlphaInterp(InterpSpeedDecreasing=7, InterpSpeedIncreasing=8, Scale=1, bClampResult=true, Value=@RigVMFunction_MathBoolToFloat.Result)
   Multiply = &RigVMFunction_MathDoubleMul(
    B=@n3.Result
    A=RigVMDispatchNode(BoneName=root, Default=1, Name=deformRig_r_upperarm_up_alpha) {template="DISPATCH_RigDispatch_GetAnimAttribute(in Name,in BoneName,in Default,out Value,out Found)"}.Value
   )
   > &RigUnit_ModifyTransforms(
    ItemToModify="((Item=(Type=Bone,Name=\"dyn_bandolier_fr_02_l\"),Transform=(Rotation=(X=0,Y=0,Z=-0.0311,W=0.9995),Translation=(X=-0.0072,Y=-0.1304,Z=-2.1974),Scale3D=(X=1,Y=1,Z=1))),(Item=(Type=Bone,Name=\"dyn_bandolier_fr_01_l\"),Transform=(Rotation=(X=-0.0006,Y=-0.0722,Z=0 ...(truncated)"
    WeightMinimum=0
    Weight=@Multiply.Result
   )
  ]}
 New Function_ContainedGraph {nodes [
   n1 = RigVMFunctionEntryNode()
   RigVMFunction_Sequence = &RigVMFunction_Sequence(ExecuteContext=@n1.ExecuteContext)
   n2 = &RigUnit_SphericalPoseReader(
    ActiveRegionScaleFactors=(PositiveWidth=0.5,NegativeWidth=0.5,PositiveHeight=0.5,NegativeHeight=0.5)
    ActiveRegionSize=0.2
    Debug=(bDrawDebug=false,bDraw2D=false,DebugScale=25,DebugSegments=20,DebugThickness=0.25)
    DriverAxis=(X=-1,Y=0,Z=0)
    DriverItem=(Type=Bone,Name="calf_l")
    FalloffRegionScaleFactors=(PositiveWidth=0,NegativeWidth=0.7,PositiveHeight=0.7,NegativeHeight=0.7)
    FalloffSize=0.45
    OptionalParentItem=(Type=Bone,Name="thigh_l")
    RotationOffset=(X=-90,Y=0,Z=60)
    ExecutePin=@RigVMFunction_Sequence.A
   )
   > &RigUnit_ModifyTransforms(
    ItemToModify=((Item=(Type=Bone,Name="calf_counter_l"),Transform=(Rotation=(X=0,Y=0,Z=0.0911,W=-0.9958),Translation=(X=-1,Y=0,Z=0),Scale3D=(X=1,Y=1,Z=1.1))))
    WeightMinimum=0
    Weight=@n2.OutputParam
   )
   > &RigUnit_ModifyTransforms(
    ItemToModify=((Item=(Type=Bone,Name="thigh_counter_l"),Transform=(Rotation=(X=0,Y=0,Z=0.0872,W=0.9962),Translation=(X=0,Y=0,Z=0),Scale3D=(X=1,Y=1,Z=1))))
    WeightMinimum=0
    Weight=@n2.OutputParam
   )
   > n3 = &RigUnit_SphericalPoseReader(
    ActiveRegionScaleFactors=(PositiveWidth=0.5,NegativeWidth=0.5,PositiveHeight=0.5,NegativeHeight=0.5)
    ActiveRegionSize=0.2
    Debug=(bDrawDebug=false,bDraw2D=false,DebugScale=25,DebugSegments=20,DebugThickness=0.25)
    DriverAxis=(X=1,Y=0,Z=0)
    DriverItem=(Type=Bone,Name="calf_r")
    FalloffRegionScaleFactors=(PositiveWidth=0,NegativeWidth=0.7,PositiveHeight=0.7,NegativeHeight=0.7)
    FalloffSize=0.45
    OptionalParentItem=(Type=Bone,Name="thigh_r")
    RotationOffset=(X=-90,Y=0,Z=-120)
   )
   > &RigUnit_ModifyTransforms(
    ItemToModify=((Item=(Type=Bone,Name="calf_counter_r"),Transform=(Rotation=(X=0,Y=0,Z=-0.1157,W=0.9933),Translation=(X=0,Y=0,Z=0),Scale3D=(X=1,Y=1,Z=1))))
    WeightMinimum=0
    Weight=@n3.OutputParam
   )
   > &RigUnit_ModifyTransforms(
    ItemToModify=((Item=(Type=Bone,Name="thigh_counter_r"),Transform=(Rotation=(X=0,Y=0,Z=0.1736,W=0.9848),Translation=(X=0,Y=0,Z=0),Scale3D=(X=1,Y=1,Z=1))))
    WeightMinimum=0
    Weight=@n3.OutputParam
   )
   RigVMFunctionReturnNode(ExecuteContext=@RigVMFunction_Sequence.B)
  ]}
 New Function_ContainedGraph {nodes [
   n1 = RigVMFunctionEntryNode()
   RigVMFunction_MathQuaternionToEuler = &RigVMFunction_MathQuaternionToEuler(
    RotationOrder=ZYX
    Value=&RigUnit_GetRelativeTransformForItem(
     Child=(Type=Bone,Name="None")
     Parent=(Type=Bone,Name="None")
     bChildInitial=False
     bParentInitial=False
     Child.Name=@n1.Child
     Parent.Name=@n1.Parent
    ).RelativeTransform.Rotation
   )
   Remap = &RigVMFunction_MathDoubleRemap(
    TargetMaximum=1
    TargetMinimum=0
    Value=@RigVMFunction_MathQuaternionToEuler.Result.Z
    SourceMinimum=@n1.SourceMin
    SourceMaximum=@n1.SourceMax
   )
   > RigVMFunctionReturnNode(Delta Z=@Remap.Result)
  ]}
 New Function_ContainedGraph {nodes [
   n1 = RigVMFunctionEntryNode()
   n2 = &RigVMFunction_ControlFlowBranch(ExecuteContext=@n1.ExecuteContext, Condition=RigVMVariableNode() {variable=isParachute}.Value)
   RigVMFunctionReturnNode(ExecuteContext=@n2.Completed)
   RigVMVariableNode(Value=(X=0,Y=0,Z=8), ExecuteContext=@n2.False) {variable=SetTranslate}
   RigVMVariableNode(Value=(X=0,Y=0,Z=5), ExecuteContext=@n2.True) {variable=SetTranslate}
  ]}
 New Function_ContainedGraph {nodes [
   n1 = RigVMFunctionEntryNode()
   > DISPATCH_RigVMDispatch_ArrayIterator = RigVMDispatchNode(Array=@n1.transform) {template="DISPATCH_RigVMDispatch_ArrayIterator(in Array,out Element,out Index,out Count,out Ratio)"}
   Multiply = &RigVMFunction_MathTransformMul(
    A=@DISPATCH_RigVMDispatch_ArrayIterator.Element.Transform
    B=&RigUnit_GetTransform(
     Item=(Type=Bone,Name="dyn_L_hair_3")
     Space=LocalSpace
     bInitial=False
     Item.Name=@DISPATCH_RigVMDispatch_ArrayIterator.Element.Bone
     Item.Type=@n1.type
    ).Transform
   ) {template="Multiply::Execute(in A,in B,out Result)"}
   > &RigUnit_SetTransform(
    Item=(Type=Bone,Name="dyn_L_hair_3")
    Space=LocalSpace
    bInitial=false
    Value=@Multiply.Result
    bPropagateToChildren=@n1.inheritsTransform
    Weight=@n1.weight
    Item.Name=@DISPATCH_RigVMDispatch_ArrayIterator.Element.Bone
    Item.Type=@n1.type
   )
   RigVMFunctionReturnNode(ExecuteContext=@DISPATCH_RigVMDispatch_ArrayIterator.Completed)
  ]}
 New Function_ContainedGraph {nodes [
   n1 = RigVMFunctionEntryNode()
   RigVMFunction_MathQuaternionToEuler = &RigVMFunction_MathQuaternionToEuler(
    RotationOrder=YXZ
    Value=&RigUnit_GetRelativeTransformForItem(
     Child=(Type=Bone,Name="None")
     Parent=(Type=Bone,Name="None")
     bChildInitial=False
     bParentInitial=False
     Child.Name=@n1.Child
     Parent.Name=@n1.Parent
    ).RelativeTransform.Rotation
   )
   Remap = &RigVMFunction_MathDoubleRemap(
    TargetMaximum=1
    TargetMinimum=0
    Value=@RigVMFunction_MathQuaternionToEuler.Result.Y
    SourceMinimum=@n1.Source Min
    SourceMaximum=@n1.Source Max
   )
   > RigVMFunctionReturnNode(Delta Y=@Remap.Result)
  ]}
 New Function_ContainedGraph {nodes [
   n1 = RigVMFunctionEntryNode()
   RigVMFunction_Sequence = &RigVMFunction_Sequence(ExecuteContext=@n1.ExecuteContext)
   &RigUnit_ParentConstraint(
    AdvancedSettings=(RotationOrderForFilter=XZY)
    Child=(Type=Bone,Name="base_bandolier_root")
    Filter=(TranslationFilter=(bX=True,bY=True,bZ=True),RotationFilter=(bX=False,bY=False,bZ=False),ScaleFilter=(bX=True,bY=True,bZ=True))
    Parents="((Item=(Type=Bone,Name=\"deform_upperarm_fr_r\"),Weight=0.0011),(Item=(Type=Bone,Name=\"deform_upperarm_up_r\"),Weight=0.0118),(Item=(Type=Bone,Name=\"upperarm_twist_01_r\"),Weight=0.0125),(Item=(Type=Bone,Name=\"spine_04\"),Weight=0.0176),(Item=(Type=Bone,Name=\"s ...(truncated)"
    Weight=1
    ExecutePin=@RigVMFunction_Sequence.A
   )
   > &RigUnit_ParentConstraint(
    AdvancedSettings=(RotationOrderForFilter=XZY)
    Child=(Type=Bone,Name="base_bandolier_root_skin")
    Filter=(TranslationFilter=(bX=True,bY=True,bZ=True),RotationFilter=(bX=False,bY=False,bZ=False),ScaleFilter=(bX=True,bY=True,bZ=True))
    Parents="((Item=(Type=Bone,Name=\"deform_upperarm_fr_r\"),Weight=0.0011),(Item=(Type=Bone,Name=\"deform_upperarm_up_r\"),Weight=0.0118),(Item=(Type=Bone,Name=\"upperarm_twist_01_r\"),Weight=0.0125),(Item=(Type=Bone,Name=\"spine_04\"),Weight=0.0176),(Item=(Type=Bone,Name=\"s ...(truncated)"
    Weight=1
   )
   > &RigUnit_ParentConstraint(
    AdvancedSettings=(RotationOrderForFilter=XZY)
    Child=(Type=Bone,Name="base_bandolier_bk_01_l")
    Filter=(TranslationFilter=(bX=True,bY=True,bZ=True),RotationFilter=(bX=False,bY=False,bZ=False),ScaleFilter=(bX=True,bY=True,bZ=True))
    Parents="((Item=(Type=Bone,Name=\"clavicle_l\"),Weight=0.0012),(Item=(Type=Bone,Name=\"spine_04\"),Weight=0.0163),(Item=(Type=Bone,Name=\"deform_upperarm_up_r\"),Weight=0.0263),(Item=(Type=Bone,Name=\"upperarm_twist_01_r\"),Weight=0.0388),(Item=(Type=Bone,Name=\"deform_uppe ...(truncated)"
    Weight=1
   )
   > &RigUnit_ParentConstraint(
    AdvancedSettings=(RotationOrderForFilter=XZY)
    Child=(Type=Bone,Name="dyn_bandolier_bk_01_l")
    Filter=(TranslationFilter=(bX=True,bY=True,bZ=True),RotationFilter=(bX=False,bY=False,bZ=False),ScaleFilter=(bX=True,bY=True,bZ=True))
    Parents="((Item=(Type=Bone,Name=\"clavicle_l\"),Weight=0.0012),(Item=(Type=Bone,Name=\"spine_04\"),Weight=0.0163),(Item=(Type=Bone,Name=\"deform_upperarm_up_r\"),Weight=0.0263),(Item=(Type=Bone,Name=\"upperarm_twist_01_r\"),Weight=0.0388),(Item=(Type=Bone,Name=\"deform_uppe ...(truncated)"
    Weight=1
   )
   > &RigUnit_ParentConstraint(
    AdvancedSettings=(RotationOrderForFilter=XZY)
    Child=(Type=Bone,Name="dyn_bandolier_bk_02_l")
    Filter=(TranslationFilter=(bX=True,bY=True,bZ=True),RotationFilter=(bX=False,bY=False,bZ=False),ScaleFilter=(bX=True,bY=True,bZ=True))
    Parents=((Item=(Type=Bone,Name="clavicle_l"),Weight=0.0306),(Item=(Type=Bone,Name="spine_04"),Weight=0.0487),(Item=(Type=Bone,Name="spine_05"),Weight=0.4338),(Item=(Type=Bone,Name="clavicle_r"),Weight=0.4868))
    Weight=1
   )
   > &RigUnit_ParentConstraint(
    AdvancedSettings=(RotationOrderForFilter=XZY)
    Child=(Type=Bone,Name="dyn_bandolier_bk_03_l")
    Filter=(TranslationFilter=(bX=True,bY=True,bZ=True),RotationFilter=(bX=False,bY=False,bZ=False),ScaleFilter=(bX=True,bY=True,bZ=True))
    Parents="((Item=(Type=Bone,Name=\"pelvis\"),Weight=0.0004),(Item=(Type=Bone,Name=\"spine_01\"),Weight=0.0263),(Item=(Type=Bone,Name=\"spine_03\"),Weight=0.0808),(Item=(Type=Bone,Name=\"spine_02\"),Weight=0.0823),(Item=(Type=Bone,Name=\"clavicle_r\"),Weight=0.0975),(Item=(Typ ...(truncated)"
    Weight=1
   )
   > &RigUnit_ParentConstraint(
    AdvancedSettings=(RotationOrderForFilter=XZY)
    Child=(Type=Bone,Name="dyn_bandolier_bk_04_l")
    Filter=(TranslationFilter=(bX=True,bY=True,bZ=True),RotationFilter=(bX=False,bY=False,bZ=False),ScaleFilter=(bX=True,bY=True,bZ=True))
    Parents="((Item=(Type=Bone,Name=\"pelvis\"),Weight=0.0145),(Item=(Type=Bone,Name=\"clavicle_r\"),Weight=0.0164),(Item=(Type=Bone,Name=\"spine_01\"),Weight=0.061),(Item=(Type=Bone,Name=\"clavicle_l\"),Weight=0.0899),(Item=(Type=Bone,Name=\"spine_05\"),Weight=0.1059),(Item=(Ty ...(truncated)"
    Weight=1
   )
   > &RigUnit_ParentConstraint(
    AdvancedSettings=(RotationOrderForFilter=XZY)
    Child=(Type=Bone,Name="dyn_bandolier_bk_05_l")
    Filter=(TranslationFilter=(bX=True,bY=True,bZ=True),RotationFilter=(bX=False,bY=False,bZ=False),ScaleFilter=(bX=True,bY=True,bZ=True))
    Parents="((Item=(Type=Bone,Name=\"spine_05\"),Weight=0.0129),(Item=(Type=Bone,Name=\"clavicle_l\"),Weight=0.0139),(Item=(Type=Bone,Name=\"spine_01\"),Weight=0.0862),(Item=(Type=Bone,Name=\"pelvis\"),Weight=0.0951),(Item=(Type=Bone,Name=\"spine_03\"),Weight=0.1276),(Item=(Typ ...(truncated)"
    Weight=1
   )
   > &RigUnit_ParentConstraint(
    AdvancedSettings=(RotationOrderForFilter=XZY)
    Child=(Type=Bone,Name="dyn_purse_bk_l")
    Filter=(TranslationFilter=(bX=True,bY=True,bZ=True),RotationFilter=(bX=False,bY=False,bZ=False),ScaleFilter=(bX=True,bY=True,bZ=True))
    Parents=((Item=(Type=Bone,Name="spine_01"),Weight=0.0626),(Item=(Type=Bone,Name="spine_03"),Weight=0.0645),(Item=(Type=Bone,Name="spine_04"),Weight=0.1087),(Item=(Type=Bone,Name="pelvis"),Weight=0.2742),(Item=(Type=Bone,Name="spine_02"),Weight=0.49))
    Weight=1
   )
   > &RigUnit_ParentConstraint(
    AdvancedSettings=(RotationOrderForFilter=XZY)
    Child=(Type=Bone,Name="dyn_bandolier_fr_01_l")
    Parents="((Item=(Type=Bone,Name=\"deform_upperarm_fr_r\"),Weight=0.007),(Item=(Type=Bone,Name=\"upperarm_twist_01_r\"),Weight=0.0086),(Item=(Type=Bone,Name=\"deform_pec_r\"),Weight=0.015),(Item=(Type=Bone,Name=\"spine_04\"),Weight=0.1107),(Item=(Type=Bone,Name=\"clavicle_r\" ...(truncated)"
    Weight=1
   )
   > &RigUnit_ParentConstraint(
    AdvancedSettings=(RotationOrderForFilter=XZY)
    Child=(Type=Bone,Name="dyn_bandolier_fr_02_l")
    Parents=((Item=(Type=Bone,Name="clavicle_r"),Weight=0.0236),(Item=(Type=Bone,Name="deform_pec_r"),Weight=0.1415),(Item=(Type=Bone,Name="spine_04"),Weight=0.3141),(Item=(Type=Bone,Name="spine_05"),Weight=0.5208))
    Weight=1
   )
   > &RigUnit_ParentConstraint(
    AdvancedSettings=(RotationOrderForFilter=XZY)
    Child=(Type=Bone,Name="dyn_bandolier_fr_03_l")
    Parents="((Item=(Type=Bone,Name=\"spine_02\"),Weight=0.0006),(Item=(Type=Bone,Name=\"deform_pec_r\"),Weight=0.0006),(Item=(Type=Bone,Name=\"spine_03\"),Weight=0.0007),(Item=(Type=Bone,Name=\"deform_pec_l\"),Weight=0.0206),(Item=(Type=Bone,Name=\"spine_05\"),Weight=0.1802),(I ...(truncated)"
    Weight=1
   )
   > &RigUnit_ParentConstraint(
    AdvancedSettings=(RotationOrderForFilter=XZY)
    Child=(Type=Bone,Name="dyn_bandolier_fr_04_l")
    Parents=((Item=(Type=Bone,Name="pelvis"),Weight=0.0096),(Item=(Type=Bone,Name="spine_01"),Weight=0.0223),(Item=(Type=Bone,Name="spine_03"),Weight=0.1048),(Item=(Type=Bone,Name="spine_02"),Weight=0.1502),(Item=(Type=Bone,Name="spine_04"),Weight=0.7131))
    Weight=1
   )
   > &RigUnit_ParentConstraint(
    AdvancedSettings=(RotationOrderForFilter=XZY)
    Child=(Type=Bone,Name="dyn_bandolier_fr_05_l")
    Parents=((Item=(Type=Bone,Name="spine_01"),Weight=0.0437),(Item=(Type=Bone,Name="pelvis"),Weight=0.0899),(Item=(Type=Bone,Name="spine_03"),Weight=0.1186),(Item=(Type=Bone,Name="spine_02"),Weight=0.3664),(Item=(Type=Bone,Name="spine_04"),Weight=0.3814))
    Weight=1
   )
   > &RigUnit_ParentConstraint(
    AdvancedSettings=(RotationOrderForFilter=XZY)
    Child=(Type=Bone,Name="dyn_purse_fr_l")
    Parents=((Item=(Type=Bone,Name="spine_01"),Weight=0.0406),(Item=(Type=Bone,Name="spine_03"),Weight=0.0683),(Item=(Type=Bone,Name="spine_04"),Weight=0.148),(Item=(Type=Bone,Name="pelvis"),Weight=0.2611),(Item=(Type=Bone,Name="spine_02"),Weight=0.4821))
    Weight=1
   )
   > Delta Rotation Z_1 = RigVMFunctionReferenceNode(Child=spine_05, Parent=spine_04, SourceMax=10, SourceMin=2)
   > &RigUnit_ModifyTransforms(
    ItemToModify=((Item=(Type=Bone,Name="dyn_bandolier_fr_02_l"),Transform=(Rotation=(X=0,Y=0,Z=0.0194,W=0.9998),Translation=(X=-0.0407,Y=2.3792,Z=0),Scale3D=(X=1,Y=1,Z=1))))
    WeightMinimum=0
    Weight=@"Delta Rotation Z_1".Delta Z
   )
   > Delta Rotation Z_4 = RigVMFunctionReferenceNode(Child=spine_05, Parent=spine_04, SourceMax=-30, SourceMin=0)
   > &RigUnit_ModifyTransforms(
    ItemToModify="((Item=(Type=Bone,Name=\"dyn_bandolier_bk_02_l\"),Transform=(Rotation=(X=0,Y=0,Z=0,W=1),Translation=(X=0,Y=0,Z=0),Scale3D=(X=1,Y=1,Z=1))),(Item=(Type=Bone,Name=\"dyn_bandolier_bk_03_l\"),Transform=(Rotation=(X=0,Y=0,Z=0,W=-1),Translation=(X=0,Y=0,Z=0),Scale3D= ...(truncated)"
    WeightMinimum=0
    Weight=@"Delta Rotation Z_4".Delta Z
   )
   > DeltaRotation Y_1 = RigVMFunctionReferenceNode(Child=clavicle_r, Parent=spine_05, Source Max=50, Source Min=80)
   > &RigUnit_ModifyTransforms(
    ItemToModify="((Item=(Type=Bone,Name=\"dyn_bandolier_fr_01_extra_l\"),Transform=(Rotation=(X=0,Y=0.144,Z=0,W=-0.9896),Translation=(X=0.7253,Y=0.4696,Z=0.5909),Scale3D=(X=1,Y=1,Z=1))),(Item=(Type=Bone,Name=\"dyn_bandolier_fr_02_l\"),Transform=(Rotation=(X=0,Y=0,Z=0.0196,W=0. ...(truncated)"
    WeightMinimum=0
    Weight=@"DeltaRotation Y_1".Delta Y
   )
   > DeltaRotation Y = RigVMFunctionReferenceNode(Child=thigh_l, Parent=pelvis, Source Max=-100, Source Min=-5)
   > &RigUnit_ModifyTransforms(
    ItemToModify=((Item=(Type=Bone,Name="dyn_purse_bk_l"),Transform=(Rotation=(X=0,Y=-0.478,Z=0,W=0.8783),Translation=(X=0,Y=0,Z=0),Scale3D=(X=1,Y=1,Z=1))))
    WeightMinimum=0
    Weight=@"DeltaRotation Y".Delta Y
   )
   > Delta Rotation Z = RigVMFunctionReferenceNode(Child=thigh_l, Parent=pelvis, SourceMax=-90, SourceMin=-3)
   > &RigUnit_ModifyTransforms(
    ItemToModify=((Item=(Type=Bone,Name="dyn_purse_bk_l"),Transform=(Rotation=(X=0,Y=-0.1305,Z=0,W=0.9914),Translation=(X=0,Y=0,Z=0),Scale3D=(X=1,Y=1,Z=1))))
    WeightMinimum=0
    Weight=@"Delta Rotation Z".Delta Z
   )
   > &RigUnit_ModifyTransforms(
    ItemToModify="((Item=(Type=Bone,Name=\"dyn_bandolier_fr_03_l\"),Transform=(Rotation=(X=0,Y=0,Z=0,W=1),Translation=(X=0,Y=0,Z=0),Scale3D=(X=1,Y=1,Z=1))),(Item=(Type=Bone,Name=\"dyn_bandolier_bk_03_l\"),Transform=(Rotation=(X=0,Y=0,Z=-0.0175,W=0.9998),Translation=(X=0,Y=0,Z=0 ...(truncated)"
    Weight=1
    WeightMinimum=0
   )
   > DeltaRotation X = RigVMFunctionReferenceNode(Child=spine_05, Parent=pelvis, SourceMax=50, SourceMin=3)
   > &RigUnit_ModifyTransforms(
    ItemToModify="((Item=(Type=Bone,Name=\"dyn_bandolier_fr_03_l\"),Transform=(Rotation=(X=-0.0097,Y=0.1018,Z=-0.0037,W=-0.9948),Translation=(X=0.7274,Y=0.0718,Z=0.2349),Scale3D=(X=1,Y=1,Z=1))),(Item=(Type=Bone,Name=\"dyn_bandolier_bk_03_l\"),Transform=(Rotation=(X=0,Y=0,Z=0,W= ...(truncated)"
    WeightMinimum=0
    Weight=@"DeltaRotation X".Delta X
   )
   > DeltaRotation X_1 = RigVMFunctionReferenceNode(Child=spine_05, Parent=pelvis, SourceMax=-50, SourceMin=-3)
   > &RigUnit_ModifyTransforms(
    ItemToModify="((Item=(Type=Bone,Name=\"dyn_bandolier_fr_03_l\"),Transform=(Rotation=(X=0,Y=0,Z=0,W=1),Translation=(X=0,Y=0,Z=0),Scale3D=(X=1,Y=1,Z=1))),(Item=(Type=Bone,Name=\"dyn_bandolier_bk_03_l\"),Transform=(Rotation=(X=0,Y=0,Z=0,W=1),Translation=(X=0,Y=0,Z=0),Scale3D=( ...(truncated)"
    WeightMinimum=0
    Weight=@"DeltaRotation X_1".Delta X
   )
   RigVMFunctionReturnNode(ExecuteContext=@RigVMFunction_Sequence.B)
  ]}
 New Function_ContainedGraph {nodes [
   Delta Rotation Z_1 = RigVMFunctionReferenceNode(Child=spine_05, ExecuteContext=(), Parent=spine_04, SourceMax=10, SourceMin=2)
   > &RigUnit_ModifyTransforms(
    ItemToModify=((Item=(Type=Bone,Name="dyn_bandolier_fr_02_l"),Transform=(Rotation=(X=0,Y=0,Z=0.0194,W=0.9998),Translation=(X=-0.0407,Y=2.3792,Z=0),Scale3D=(X=1,Y=1,Z=1))))
    WeightMinimum=0
    Weight=@"Delta Rotation Z_1".Delta Z
   )
   > Delta Rotation Z_4 = RigVMFunctionReferenceNode(Child=spine_05, Parent=spine_04, SourceMax=-30, SourceMin=0)
   > &RigUnit_ModifyTransforms(
    ItemToModify="((Item=(Type=Bone,Name=\"dyn_bandolier_bk_02_l\"),Transform=(Rotation=(X=0,Y=0,Z=0,W=1),Translation=(X=0,Y=0,Z=0),Scale3D=(X=1,Y=1,Z=1))),(Item=(Type=Bone,Name=\"dyn_bandolier_bk_03_l\"),Transform=(Rotation=(X=0,Y=0,Z=0,W=-1),Translation=(X=0,Y=-2.5837,Z=0),Sc ...(truncated)"
    WeightMinimum=0
    Weight=@"Delta Rotation Z_4".Delta Z
   )
   > DeltaRotation Y_1 = RigVMFunctionReferenceNode(Child=clavicle_r, Parent=spine_05, Source Max=50, Source Min=80)
   > &RigUnit_ModifyTransforms(
    ItemToModify="((Item=(Type=Bone,Name=\"dyn_bandolier_fr_01_extra_l\"),Transform=(Rotation=(X=0,Y=0.144,Z=0,W=-0.9896),Translation=(X=0.7253,Y=0.4696,Z=0.5909),Scale3D=(X=1,Y=1,Z=1))),(Item=(Type=Bone,Name=\"dyn_bandolier_fr_02_l\"),Transform=(Rotation=(X=0,Y=0,Z=0.0196,W=0. ...(truncated)"
    WeightMinimum=0
    Weight=@"DeltaRotation Y_1".Delta Y
   )
   > Delta Rotation Z_5 = RigVMFunctionReferenceNode(Child=spine_01, Parent=pelvis, SourceMax=30, SourceMin=20)
   n1 = RigVMFunctionEntryNode()
   RigVMFunction_Sequence = &RigVMFunction_Sequence(ExecuteContext=@n1.ExecuteContext)
   AlphaInterp = &RigVMFunction_AlphaInterp(InterpSpeedDecreasing=5, InterpSpeedIncreasing=5, Scale=1, bClampResult=true, Value=RigVMVariableNode() {variable=RailSeated}.Value)
   &RigUnit_ParentConstraint(
    AdvancedSettings=(RotationOrderForFilter=XZY)
    Child=(Type=Bone,Name="base_bandolier_root")
    Parents="((Item=(Type=Bone,Name=\"deform_upperarm_fr_r\"),Weight=0.0011),(Item=(Type=Bone,Name=\"deform_upperarm_up_r\"),Weight=0.0118),(Item=(Type=Bone,Name=\"upperarm_twist_01_r\"),Weight=0.0125),(Item=(Type=Bone,Name=\"spine_04\"),Weight=0.0176),(Item=(Type=Bone,Name=\"s ...(truncated)"
    Weight=@AlphaInterp.Result
    ExecutePin=@RigVMFunction_Sequence.A
   )
   > &RigUnit_ParentConstraint(
    AdvancedSettings=(RotationOrderForFilter=XZY)
    Child=(Type=Bone,Name="base_bandolier_root_skin")
    Parents="((Item=(Type=Bone,Name=\"deform_upperarm_fr_r\"),Weight=0.0011),(Item=(Type=Bone,Name=\"deform_upperarm_up_r\"),Weight=0.0118),(Item=(Type=Bone,Name=\"upperarm_twist_01_r\"),Weight=0.0125),(Item=(Type=Bone,Name=\"spine_04\"),Weight=0.0176),(Item=(Type=Bone,Name=\"s ...(truncated)"
    Weight=@AlphaInterp.Result
   )
   > &RigUnit_ParentConstraint(
    AdvancedSettings=(RotationOrderForFilter=XZY)
    Child=(Type=Bone,Name="base_bandolier_bk_01_l")
    Parents="((Item=(Type=Bone,Name=\"clavicle_l\"),Weight=0.0012),(Item=(Type=Bone,Name=\"spine_04\"),Weight=0.0163),(Item=(Type=Bone,Name=\"deform_upperarm_up_r\"),Weight=0.0263),(Item=(Type=Bone,Name=\"upperarm_twist_01_r\"),Weight=0.0388),(Item=(Type=Bone,Name=\"deform_uppe ...(truncated)"
    Weight=@AlphaInterp.Result
   )
   > &RigUnit_ParentConstraint(
    AdvancedSettings=(RotationOrderForFilter=XZY)
    Child=(Type=Bone,Name="dyn_bandolier_bk_01_l")
    Parents="((Item=(Type=Bone,Name=\"clavicle_l\"),Weight=0.0012),(Item=(Type=Bone,Name=\"spine_04\"),Weight=0.0163),(Item=(Type=Bone,Name=\"deform_upperarm_up_r\"),Weight=0.0263),(Item=(Type=Bone,Name=\"upperarm_twist_01_r\"),Weight=0.0388),(Item=(Type=Bone,Name=\"deform_uppe ...(truncated)"
    Weight=@AlphaInterp.Result
   )
   > &RigUnit_ParentConstraint(
    AdvancedSettings=(RotationOrderForFilter=XZY)
    Child=(Type=Bone,Name="dyn_bandolier_bk_02_l")
    Parents=((Item=(Type=Bone,Name="clavicle_l"),Weight=0.0306),(Item=(Type=Bone,Name="spine_04"),Weight=0.0487),(Item=(Type=Bone,Name="spine_05"),Weight=0.4338),(Item=(Type=Bone,Name="clavicle_r"),Weight=0.4868))
    Weight=@AlphaInterp.Result
   )
   > &RigUnit_ParentConstraint(
    AdvancedSettings=(RotationOrderForFilter=XZY)
    Child=(Type=Bone,Name="dyn_bandolier_bk_03_l")
    Parents="((Item=(Type=Bone,Name=\"pelvis\"),Weight=0.0004),(Item=(Type=Bone,Name=\"spine_01\"),Weight=0.0263),(Item=(Type=Bone,Name=\"spine_03\"),Weight=0.0808),(Item=(Type=Bone,Name=\"spine_02\"),Weight=0.0823),(Item=(Type=Bone,Name=\"clavicle_r\"),Weight=0.0975),(Item=(Typ ...(truncated)"
    Weight=@AlphaInterp.Result
   )
   > &RigUnit_ParentConstraint(
    AdvancedSettings=(RotationOrderForFilter=XZY)
    Child=(Type=Bone,Name="dyn_bandolier_bk_04_l")
    Parents="((Item=(Type=Bone,Name=\"pelvis\"),Weight=0.0145),(Item=(Type=Bone,Name=\"clavicle_r\"),Weight=0.0164),(Item=(Type=Bone,Name=\"spine_01\"),Weight=0.061),(Item=(Type=Bone,Name=\"clavicle_l\"),Weight=0.0899),(Item=(Type=Bone,Name=\"spine_05\"),Weight=0.1059),(Item=(Ty ...(truncated)"
    Weight=@AlphaInterp.Result
   )
   > &RigUnit_ParentConstraint(
    AdvancedSettings=(RotationOrderForFilter=XZY)
    Child=(Type=Bone,Name="dyn_bandolier_bk_05_l")
    Parents="((Item=(Type=Bone,Name=\"spine_05\"),Weight=0.0129),(Item=(Type=Bone,Name=\"clavicle_l\"),Weight=0.0139),(Item=(Type=Bone,Name=\"spine_01\"),Weight=0.0862),(Item=(Type=Bone,Name=\"pelvis\"),Weight=0.0951),(Item=(Type=Bone,Name=\"spine_03\"),Weight=0.1276),(Item=(Typ ...(truncated)"
    Weight=@AlphaInterp.Result
   )
   > &RigUnit_ParentConstraint(
    AdvancedSettings=(RotationOrderForFilter=XZY)
    Child=(Type=Bone,Name="dyn_purse_bk_l")
    Parents=((Item=(Type=Bone,Name="spine_01"),Weight=0.0626),(Item=(Type=Bone,Name="spine_03"),Weight=0.0645),(Item=(Type=Bone,Name="spine_04"),Weight=0.1087),(Item=(Type=Bone,Name="pelvis"),Weight=0.2742),(Item=(Type=Bone,Name="spine_02"),Weight=0.49))
    Weight=@AlphaInterp.Result
   )
   > &RigUnit_ParentConstraint(
    AdvancedSettings=(RotationOrderForFilter=XZY)
    Child=(Type=Bone,Name="dyn_bandolier_fr_01_l")
    Parents="((Item=(Type=Bone,Name=\"deform_upperarm_fr_r\"),Weight=0.007),(Item=(Type=Bone,Name=\"upperarm_twist_01_r\"),Weight=0.0086),(Item=(Type=Bone,Name=\"deform_pec_r\"),Weight=0.015),(Item=(Type=Bone,Name=\"spine_04\"),Weight=0.1107),(Item=(Type=Bone,Name=\"clavicle_r\" ...(truncated)"
    Weight=@AlphaInterp.Result
   )
   > &RigUnit_ParentConstraint(
    AdvancedSettings=(RotationOrderForFilter=XZY)
    Child=(Type=Bone,Name="dyn_bandolier_fr_02_l")
    Parents=((Item=(Type=Bone,Name="clavicle_r"),Weight=0.0236),(Item=(Type=Bone,Name="deform_pec_r"),Weight=0.1415),(Item=(Type=Bone,Name="spine_04"),Weight=0.3141),(Item=(Type=Bone,Name="spine_05"),Weight=0.5208))
    Weight=@AlphaInterp.Result
   )
   > &RigUnit_ParentConstraint(
    AdvancedSettings=(RotationOrderForFilter=XZY)
    Child=(Type=Bone,Name="dyn_bandolier_fr_03_l")
    Parents="((Item=(Type=Bone,Name=\"spine_02\"),Weight=0.0006),(Item=(Type=Bone,Name=\"deform_pec_r\"),Weight=0.0006),(Item=(Type=Bone,Name=\"spine_03\"),Weight=0.0007),(Item=(Type=Bone,Name=\"deform_pec_l\"),Weight=0.0206),(Item=(Type=Bone,Name=\"spine_05\"),Weight=0.1802),(I ...(truncated)"
    Weight=@AlphaInterp.Result
   )
   > &RigUnit_ParentConstraint(
    AdvancedSettings=(RotationOrderForFilter=XZY)
    Child=(Type=Bone,Name="dyn_bandolier_fr_04_l")
    Parents=((Item=(Type=Bone,Name="pelvis"),Weight=0.0096),(Item=(Type=Bone,Name="spine_01"),Weight=0.0223),(Item=(Type=Bone,Name="spine_03"),Weight=0.1048),(Item=(Type=Bone,Name="spine_02"),Weight=0.1502),(Item=(Type=Bone,Name="spine_04"),Weight=0.7131))
    Weight=@AlphaInterp.Result
   )
   > &RigUnit_ParentConstraint(
    AdvancedSettings=(RotationOrderForFilter=XZY)
    Child=(Type=Bone,Name="dyn_bandolier_fr_05_l")
    Parents=((Item=(Type=Bone,Name="spine_01"),Weight=0.0437),(Item=(Type=Bone,Name="pelvis"),Weight=0.0899),(Item=(Type=Bone,Name="spine_03"),Weight=0.1186),(Item=(Type=Bone,Name="spine_02"),Weight=0.3664),(Item=(Type=Bone,Name="spine_04"),Weight=0.3814))
    Weight=@AlphaInterp.Result
   )
   > &RigUnit_ParentConstraint(
    AdvancedSettings=(RotationOrderForFilter=XZY)
    Child=(Type=Bone,Name="dyn_purse_fr_l")
    Parents=((Item=(Type=Bone,Name="spine_01"),Weight=0.0406),(Item=(Type=Bone,Name="spine_03"),Weight=0.0683),(Item=(Type=Bone,Name="spine_04"),Weight=0.148),(Item=(Type=Bone,Name="pelvis"),Weight=0.2611),(Item=(Type=Bone,Name="spine_02"),Weight=0.4821))
    Weight=@AlphaInterp.Result
   )
   RigVMFunctionReturnNode(ExecuteContext=@RigVMFunction_Sequence.B)
  ]}
 New Function_ContainedGraph {nodes [
   n1 = RigVMFunctionEntryNode()
   RigVMFunction_MathQuaternionToEuler = &RigVMFunction_MathQuaternionToEuler(
    RotationOrder=ZYX
    Value=&RigUnit_GetRelativeTransformForItem(
     Child=(Type=Bone,Name="None")
     Parent=(Type=Bone,Name="None")
     bChildInitial=False
     bParentInitial=False
     Child.Name=@n1.Child
     Parent.Name=@n1.Parent
    ).RelativeTransform.Rotation
   )
   Remap = &RigVMFunction_MathDoubleRemap(
    TargetMaximum=1
    TargetMinimum=0
    Value=@RigVMFunction_MathQuaternionToEuler.Result.X
    SourceMinimum=@n1.SourceMin
    SourceMaximum=@n1.SourceMax
   )
   > RigVMFunctionReturnNode(Delta X=@Remap.Result)
  ]}
 RigVMFunctionLibrary {nodes [
   AlterTransform = RigVMCollapseNode(ExecuteContext=(), transform=())
   Constraints = RigVMCollapseNode(ExecuteContext=())
   Delta Rotation Z = RigVMCollapseNode(ExecuteContext=())
   DeltaRotation X = RigVMCollapseNode(ExecuteContext=())
   DeltaRotation Y = RigVMCollapseNode(ExecuteContext=())
   FixCrouchSkydive = RigVMCollapseNode(ExecuteContext=())
   FixLegCollisionOnCrouch = RigVMCollapseNode(ExecuteContext=())
   SetupNORban = RigVMCollapseNode(ExecuteContext=())
   SetupRail = RigVMCollapseNode(ExecuteContext=())
   ShoulderVolume = RigVMCollapseNode(ExecuteContext=())
  ]}
 RigVMFunction_Sequence_1_ContainedGraph {nodes [
   n1 = RigVMFunctionEntryNode()
   RigVMFunction_Sequence = &RigVMFunction_Sequence(ExecuteContext=@n1.ExecuteContext)
   n2 = &RigVMFunction_Sequence(ExecuteContext=@RigVMFunction_Sequence.B)
   RigVMFunctionReturnNode(A=@RigVMFunction_Sequence.A, B=@n2.A, C=@n2.B)
  ]}
 RigVMModel {nodes [
   RigUnit_BeginExecution = &RigUnit_BeginExecution()
   RigVMFunction_Sequence = &RigVMFunction_Sequence(ExecuteContext=@RigUnit_BeginExecution.ExecutePin)
   ShoulderVolume = RigVMFunctionReferenceNode(ExecuteContext=@RigVMFunction_Sequence.A)
   > FN_DeformRig_FNC = RigVMFunctionReferenceNode(
    DeformRig_ Curve="(deformRig_l_upperarm_up_alpha_72_#1=True,deformRig_r_upperarm_up_alpha_73_#2=True,deformRig_l_upperarm_fr_alpha_74_#3=False,deformRig_r_upperarm_fr_alpha_75_CB51B80C ...(truncated)"
    DeformRig_Bool="(upperarm_1_#4=true,elbow_3_#5=true,wrist_7_#6=true,groin_8_#7=true,glute_17_#8=true,knee_9_40C408324CA261 ...(truncated)"
    DeformRig_Debug=(upperarm_debug_18_#4=False,wrist_debug_20_#6=False,lat_debug_26_#9=False,groinGlute_debug_27_#7=False)
    DeformRig_Value="(upperarm_up_18_#10=(Rotation=(X=0,Y=0,Z=0,W=1),Translation=(X=0,Y=0,Z=8),Scale3D=(X=0.7,Y=1,Z=1)),upperarm_fwd_fr_19_#11=(Rotation=(X=0,Y=0,Z=0,W=1),Translation=(X=0,Y=10,Z=0),Scale3D=(X=0.5,Y=1,Z= ...(truncated)"
    DeformRig_Value.upperarm_up_18_#10.Translation=RigVMVariableNode() {variable=SetTranslate}.Value
   )
   > Constraints = RigVMFunctionReferenceNode()
   > FixCrouchSkydive = RigVMFunctionReferenceNode()
   > FixLegCollisionOnCrouch = RigVMFunctionReferenceNode()
   n1 = &RigUnit_SphericalPoseReader(
    ActiveRegionScaleFactors=(PositiveWidth=1,NegativeWidth=1,PositiveHeight=1,NegativeHeight=0.5)
    ActiveRegionSize=0.17
    Debug=(bDrawDebug=false,bDraw2D=False,bDrawLocalAxes=False,DebugScale=25,DebugSegments=20,DebugThickness=0.25)
    DriverAxis=(X=-1,Y=0,Z=0)
    DriverItem=(Type=Bone,Name="upperarm_r")
    FalloffRegionScaleFactors=(PositiveWidth=1,NegativeWidth=1,PositiveHeight=1,NegativeHeight=0.5)
    FalloffSize=0.27
    FlipHeightScaling=False
    FlipWidthScaling=False
    OptionalParentItem=(Type=Bone,Name="spine_05")
    RotationOffset=(X=0,Y=180,Z=200)
    ExecutePin=@RigVMFunction_Sequence.B
   )
   > &RigUnit_ModifyTransforms(
    ItemToModify=((Item=(Type=Bone,Name="base_bandolier_root"),Transform=(Rotation=(X=0,Y=0,Z=0,W=1),Translation=(X=0,Y=0,Z=1),Scale3D=(X=1,Y=1,Z=1))))
    WeightMinimum=0
    Weight=@n1.OutputParam
   )
   Remap = &RigVMFunction_MathDoubleRemap(SourceMaximum=1, SourceMinimum=0, TargetMaximum=0.82, TargetMinimum=1, Value=@n1.OutputParam)
   > RigVMVariableNode(Value=@Remap.Result) {variable=bandolier constraint alpha}
   &RigUnit_PrepareForExecution()
  ]}
]
hierarchy [
 root {loc=<0 0 0>, rot=<0 0 0 1>}
 pelvis {parent=root, loc=<0 -0.5301 98.0571>, rot=<0.0127 0.707 -0.0127 -0.707>}
 spine_01 {parent=pelvis, loc=<6.1208 -0.9328 0>, rot=<0 0 0.034 -0.9994>}
 spine_02 {parent=spine_01, loc=<6.5997 -0.2095 0>, rot=<0 0 0.0238 -0.9997>}
 spine_03 {parent=spine_02, loc=<6.4914 -0.1237 0>, rot=<0 0 0.0149 -0.9999>}
 spine_04 {parent=spine_03, loc=<6.9598 0.0794 0>, rot=<0 0 -0.0132 -0.9999>}
 spine_05 {parent=spine_04, loc=<6.9987 0 0>, rot=<0 0 -0.0015 -1>}
 neck_01 {parent=spine_05, loc=<11.3138 0.1529 0>, rot=<0 0 -0.2003 -0.9797>}
 neck_02 {parent=neck_01, loc=<4.5627 0 0>, rot=<0 0 0.0504 -0.9987>}
 head {parent=neck_02, loc=<5.3914 0 0>, rot=<0 0 0.115 -0.9934>}
 clavicle_l {parent=spine_05, loc=<5.4831 -0.1053 -3.2946>, rot=<0.0465 -0.7289 -0.009 -0.6829>}
 upperarm_l {parent=clavicle_l, loc=<9.8658 -1.2136 -0.4105>, rot=<0.0164 -0.3245 -0.0269 -0.9454>}
 lowerarm_l {parent=upperarm_l, loc=<26.465 -0.8345 0.259>, rot=<0.0036 0.0161 -0.2298 -0.9731>}
 hand_l {parent=lowerarm_l, loc=<22.7355 -0.3203 -0.6105>, rot=<-0.6784 -0.0223 0.0735 -0.7307>}
 thumb_01_l {parent=hand_l, loc=<1.3118 -0.3191 -2.459>, rot=<0.6546 -0.4042 0.0664 -0.6354>}
 thumb_02_l {parent=thumb_01_l, loc=<5.9336 -0.0448 0.0401>, rot=<-0.0037 0.0213 0.298 -0.9543>}
 thumb_03_l {parent=thumb_02_l, loc=<2.2223 0.3682 -0.0907>, rot=<0.0118 -0.013 0.0066 -0.9998>}
 index_metacarpal_l {parent=hand_l, loc=<1.6535 0.0183 -2.96>, rot=<0.0158 -0.0124 -0.0173 -0.9996>}
 index_01_l {parent=index_metacarpal_l, loc=<8.3008 0.0357 -0.0524>, rot=<-0.0044 0.0485 0.0924 -0.9945>}
 index_02_l {parent=index_01_l, loc=<3.3587 -0.3004 0.0452>, rot=<-0.002 0.0027 0.3225 -0.9466>}
 index_03_l {parent=index_02_l, loc=<2.5578 -0.123 -0.0252>, rot=<-0.002 0.0067 0.2558 -0.9667>}
 middle_metacarpal_l {parent=hand_l, loc=<2.1856 -0.1725 -0.5984>, rot=<0.006 0.0168 -0.0329 -0.9993>}
 middle_01_l {parent=middle_metacarpal_l, loc=<7.3275 -0.2447 0.0129>, rot=<-0.0208 0.0042 0.0858 -0.9961>}
 middle_02_l {parent=middle_01_l, loc=<4.0123 0.1401 0.1175>, rot=<-0.0218 0.0258 0.3512 -0.9357>}
 middle_03_l {parent=middle_02_l, loc=<2.6077 0.0534 0.1373>, rot=<0.0082 0.0188 0.1708 -0.9851>}
 ring_metacarpal_l {parent=hand_l, loc=<2.2074 -0.3625 1.3638>, rot=<-0.0284 0.0722 -0.0458 -0.9959>}
 ring_01_l {parent=ring_metacarpal_l, loc=<7.0295 -0.6197 -0.1446>, rot=<-0.0489 -0.0252 0.1892 -0.9804>}
 ring_02_l {parent=ring_01_l, loc=<3.6174 0.4647 -0.1027>, rot=<0 0 0.2806 -0.9598>}
 ring_03_l {parent=ring_02_l, loc=<2.4152 0.114 0.0199>, rot=<0 0 0.1189 -0.9929>}
 pinky_metacarpal_l {parent=hand_l, loc=<1.7616 -0.9753 3.3814>, rot=<-0.0397 0.0814 -0.069 -0.9935>}
 pinky_01_l {parent=pinky_metacarpal_l, loc=<7.1709 -0.4831 -0.1094>, rot=<-0.0654 -0.0035 0.2242 -0.9723>}
 pinky_02_l {parent=pinky_01_l, loc=<2.8274 0.0893 -0.1043>, rot=<0.0037 0.0011 0.2854 -0.9584>}
 pinky_03_l {parent=pinky_02_l, loc=<2.2046 0.0307 -0.0078>, rot=<-0.0002 0.0098 0.1445 -0.9895>}
 deform_wrist_dn_base_l {parent=hand_l, loc=<0 -1.5326 0>, rot=<0 0 0 -1>}
 deform_wrist_dn_l {parent=deform_wrist_dn_base_l, loc=<0 0 0>, rot=<0 0 0 1>}
 deform_wrist_up_base_l {parent=hand_l, loc=<0 2.3009 0>, rot=<0 0 0 -1>}
 deform_wrist_up_l {parent=deform_wrist_up_base_l, loc=<0 0 0>, rot=<0 0 0 1>}
 lowerarm_twist_02_l {parent=lowerarm_l, loc=<10.6707 0 0>, rot=<0 0 0 1>}
 lowerarm_twist_01_l {parent=lowerarm_l, loc=<18.316 0 0>, rot=<0 0 0 1>}
 deform_elbow_in_base_l {parent=lowerarm_l, loc=<0.1834 3.6062 -0.2572>, rot=<0 0 0 -1>}
 deform_elbow_in_l {parent=deform_elbow_in_base_l, loc=<0 0 0>, rot=<0 0 0 1>}
 deform_elbow_out_base_l {parent=lowerarm_l, loc=<0 -4 0>, rot=<0 0 0 1>}
 deform_elbow_out_l {parent=deform_elbow_out_base_l, loc=<0 0 0>, rot=<0 0 0 1>}
 upperarm_twist_01_l {parent=upperarm_l, loc=<7.2135 -0.1134 0>, rot=<0 0 0 1>}
 deform_upperarm_bk_base_l {parent=upperarm_twist_01_l, loc=<-6.4053 -4.3746 0.314>, rot=<0 0 0 -1>}
 deform_upperarm_bk_l {parent=deform_upperarm_bk_base_l, loc=<0 0 0>, rot=<0 0 0 1>}
 deform_upperarm_fr_base_l {parent=upperarm_twist_01_l, loc=<-5.6312 5.092 0.3354>, rot=<0 0 0 -1>}
 deform_upperarm_fr_l {parent=deform_upperarm_fr_base_l, loc=<0 0 0>, rot=<0 0 0 1>}
 deform_upperarm_up_base_l {parent=upperarm_twist_01_l, loc=<-8.891 0.4825 3.9517>, rot=<0 0 0 -1>}
 deform_upperarm_up_l {parent=deform_upperarm_up_base_l, loc=<0 0 0>, rot=<0 0 0 1>}
 upperarm_twist_02_l {parent=upperarm_l, loc=<19.7998 -0.1134 0>, rot=<0 0 0 1>}
 clavicle_r {parent=spine_05, loc=<5.4831 -0.1053 3.2946>, rot=<0.7289 0.0465 -0.6829 0.009>}
 upperarm_r {parent=clavicle_r, loc=<-9.8658 1.2136 0.4105>, rot=<0.0164 -0.3245 -0.0269 -0.9454>}
 lowerarm_r {parent=upperarm_r, loc=<-26.465 0.8345 -0.259>, rot=<0.0036 0.0161 -0.2298 -0.9731>}
 hand_r {parent=lowerarm_r, loc=<-22.7355 0.3203 0.6106>, rot=<-0.6784 -0.0223 0.0735 -0.7307>}
 thumb_01_r {parent=hand_r, loc=<-1.3118 0.3191 2.459>, rot=<0.6546 -0.4042 0.0664 -0.6354>}
 thumb_02_r {parent=thumb_01_r, loc=<-5.9336 0.0448 -0.0401>, rot=<-0.0037 0.0213 0.298 -0.9543>}
 thumb_03_r {parent=thumb_02_r, loc=<-2.2223 -0.3682 0.0907>, rot=<0.0118 -0.013 0.0066 -0.9998>}
 index_metacarpal_r {parent=hand_r, loc=<-1.6535 -0.0183 2.96>, rot=<0.0158 -0.0124 -0.0173 -0.9996>}
 index_01_r {parent=index_metacarpal_r, loc=<-8.3008 -0.0357 0.0524>, rot=<-0.0044 0.0485 0.0924 -0.9945>}
 index_02_r {parent=index_01_r, loc=<-3.3587 0.3004 -0.0452>, rot=<-0.002 0.0027 0.3225 -0.9466>}
 index_03_r {parent=index_02_r, loc=<-2.5578 0.123 0.0252>, rot=<-0.002 0.0067 0.2558 -0.9667>}
 middle_metacarpal_r {parent=hand_r, loc=<-2.1856 0.1725 0.5984>, rot=<0.006 0.0168 -0.0329 -0.9993>}
 middle_01_r {parent=middle_metacarpal_r, loc=<-7.3275 0.2447 -0.0129>, rot=<-0.0208 0.0042 0.0858 -0.9961>}
 middle_02_r {parent=middle_01_r, loc=<-4.0123 -0.1401 -0.1175>, rot=<-0.0218 0.0258 0.3512 -0.9357>}
 middle_03_r {parent=middle_02_r, loc=<-2.6077 -0.0534 -0.1373>, rot=<0.0082 0.0188 0.1708 -0.9851>}
 ring_metacarpal_r {parent=hand_r, loc=<-2.2074 0.3625 -1.3638>, rot=<-0.0284 0.0722 -0.0458 -0.9959>}
 ring_01_r {parent=ring_metacarpal_r, loc=<-7.0295 0.6197 0.1446>, rot=<-0.0489 -0.0252 0.1892 -0.9804>}
 ring_02_r {parent=ring_01_r, loc=<-3.6174 -0.4647 0.1027>, rot=<0 0 0.2806 -0.9598>}
 ring_03_r {parent=ring_02_r, loc=<-2.4152 -0.114 -0.0199>, rot=<0 0 0.1189 -0.9929>}
 pinky_metacarpal_r {parent=hand_r, loc=<-1.7616 0.9753 -3.3814>, rot=<-0.0397 0.0814 -0.069 -0.9935>}
 pinky_01_r {parent=pinky_metacarpal_r, loc=<-7.1709 0.4831 0.1094>, rot=<-0.0654 -0.0035 0.2242 -0.9723>}
 pinky_02_r {parent=pinky_01_r, loc=<-2.8274 -0.0893 0.1043>, rot=<0.0037 0.0011 0.2854 -0.9584>}
 pinky_03_r {parent=pinky_02_r, loc=<-2.2046 -0.0307 0.0078>, rot=<-0.0002 0.0098 0.1445 -0.9895>}
 deform_wrist_dn_base_r {parent=hand_r, loc=<-0.0001 1.5328 0>, rot=<0 0 0 -1>}
 deform_wrist_dn_r {parent=deform_wrist_dn_base_r, loc=<0 0 0>, rot=<0 0 0 1>}
 deform_wrist_up_base_r {parent=hand_r, loc=<-0.0002 -2.3007 -0.0001>, rot=<0 0 0 -1>}
 deform_wrist_up_r {parent=deform_wrist_up_base_r, loc=<0 0 0>, rot=<0 0 0 1>}
 lowerarm_twist_02_r {parent=lowerarm_r, loc=<-10.6707 0 0>, rot=<0 0 0 -1>}
 lowerarm_twist_01_r {parent=lowerarm_r, loc=<-18.316 0 0>, rot=<0 0 0 -1>}
 deform_elbow_in_base_r {parent=lowerarm_r, loc=<-0.1834 -3.6062 0.2572>, rot=<0 0 0 -1>}
 deform_elbow_in_r {parent=deform_elbow_in_base_r, loc=<0 0 0>, rot=<0 0 0 1>}
 deform_elbow_out_base_r {parent=lowerarm_r, loc=<0 4 -0.0001>, rot=<0 0 0 -1>}
 deform_elbow_out_r {parent=deform_elbow_out_base_r, loc=<0 0 0>, rot=<0 0 0 1>}
 upperarm_twist_01_r {parent=upperarm_r, loc=<-7.2135 0.1134 0>, rot=<0 0 0 -1>}
 deform_upperarm_bk_base_r {parent=upperarm_twist_01_r, loc=<6.4053 4.3746 -0.314>, rot=<0 0 0 -1>}
 deform_upperarm_bk_r {parent=deform_upperarm_bk_base_r, loc=<0 0 0>, rot=<0 0 0 1>}
 deform_upperarm_fr_base_r {parent=upperarm_twist_01_r, loc=<5.6312 -5.092 -0.3354>, rot=<0 0 0 -1>}
 deform_upperarm_fr_r {parent=deform_upperarm_fr_base_r, loc=<0 0 0>, rot=<0 0 0 1>}
 deform_upperarm_up_base_r {parent=upperarm_twist_01_r, loc=<8.891 -0.4825 -3.9517>, rot=<0 0 0 -1>}
 deform_upperarm_up_r {parent=deform_upperarm_up_base_r, loc=<0 0 0>, rot=<0 0 0 1>}
 upperarm_twist_02_r {parent=upperarm_r, loc=<-19.7998 0.1134 0>, rot=<0 0 0 -1>}
 attach_backpack {parent=spine_05, loc=<-0.1238 -13.4729 0>, rot=<-0.0093 -0.707 0.0101 -0.7071>}
 Backpack_BR {parent=attach_backpack, type=Null, loc=<0 3 0>, rot=<0 0 0 1>}
 attach_cape {parent=spine_05, loc=<9.0164 -11.7139 0>, rot=<0.0616 -0.7045 -0.0608 -0.7045>}
 cape {parent=attach_cape, type=Null, loc=<0 1.5 0>, rot=<0 0 0 1>}
 base_bandolier_root {parent=spine_05, loc=<9.2007 -1.4179 9.2625>, rot=<0.0385 -0.9587 -0.0113 -0.2817>}
 base_bandolier_root_skin {parent=base_bandolier_root, loc=<0 0 0>, rot=<0 0 0 1>}
 base_bandolier_bk_01_l {parent=base_bandolier_root, loc=<2.1907 -4.3909 -0.2334>, rot=<0 0 0.1595 -0.9872>}
 dyn_bandolier_bk_01_l {parent=base_bandolier_bk_01_l, loc=<0 0 0>, rot=<0 0 0 1>}
 dyn_bandolier_bk_02_l {parent=dyn_bandolier_bk_01_l, loc=<5.7937 0.2181 0.1298>, rot=<0.007 0.0291 -0.1465 -0.9888>}
 dyn_bandolier_bk_03_l {parent=dyn_bandolier_bk_02_l, loc=<10.5876 1.053 0.0057>, rot=<0 0 -0.1522 -0.9883>}
 dyn_bandolier_bk_04_l {parent=dyn_bandolier_bk_03_l, loc=<6.9066 -0.1354 0>, rot=<0 0 -0.1033 -0.9946>}
 dyn_bandolier_bk_05_l {parent=dyn_bandolier_bk_04_l, loc=<7.4309 -0.195 0>, rot=<-0.0207 -0.0077 -0.1562 -0.9875>}
 dyn_purse_bk_l {parent=dyn_bandolier_bk_05_l, loc=<4.3963 0 0>, rot=<-0.0295 -0.0726 0.4014 -0.9125>}
 dyn_pendant_01_l {parent=dyn_purse_bk_l, loc=<8.4505 2.8185 1.1113>, rot=<0 0.0409 0 -0.9992>}
 dyn_pendant_02_l {parent=dyn_pendant_01_l, loc=<2.4994 0.0194 0.0979>, rot=<0 0 0 1>}
 dyn_bandolier_bk_02_extra_l {parent=dyn_bandolier_bk_02_l, loc=<0 0 0>, rot=<0 0 0 1>}
 dyn_bandolier_bk_01_extra_l {parent=dyn_bandolier_bk_01_l, loc=<0 0 0>, rot=<0 0 0 1>}
 dyn_bandolier_bk_01_extra_02_l {parent=dyn_bandolier_bk_01_extra_l, loc=<0 0 0>, rot=<0 0 0 1>}
 dyn_bandolier_fr_01_l {parent=base_bandolier_root, loc=<3.4059 6.8211 0.1781>, rot=<-0.0058 0.0109 -0.3421 -0.9396>}
 dyn_bandolier_fr_02_l {parent=dyn_bandolier_fr_01_l, loc=<8.2453 -0.0945 0.32>, rot=<0.0115 -0.0296 0.2206 -0.9748>}
 dyn_bandolier_fr_03_l {parent=dyn_bandolier_fr_02_l, loc=<7.9429 -0.0567 0>, rot=<0 0 0.2025 -0.9793>}
 dyn_bandolier_fr_04_l {parent=dyn_bandolier_fr_03_l, loc=<7.7784 0 0>, rot=<0 0 0.048 -0.9988>}
 dyn_bandolier_fr_05_l {parent=dyn_bandolier_fr_04_l, loc=<6.3639 -0.1675 0>, rot=<-0.0701 0.1166 0.1831 -0.9736>}
 dyn_purse_fr_l {parent=dyn_bandolier_fr_05_l, loc=<4.6456 0 0>, rot=<0.0608 -0.1814 -0.2855 -0.9391>}
 dyn_bandolier_fr_02_extra_l {parent=dyn_bandolier_fr_02_l, loc=<0 0 0>, rot=<0 0 0 1>}
 dyn_bandolier_fr_01_extra_l {parent=dyn_bandolier_fr_01_l, loc=<0 0 0>, rot=<0 0 0 1>}
 deform_pec_aim_base_l {parent=spine_04, loc=<3.2456 9.3309 -1.2237>, rot=<0.0603 -0.4033 0.3318 -0.8507>}
 deform_pec_aim_l {parent=deform_pec_aim_base_l, loc=<0 0 0>, rot=<0 0 0 1>}
 deform_pec_base_l {parent=deform_pec_aim_l, loc=<9.3665 0 0>, rot=<0 0 0 1>}
 deform_pec_l {parent=deform_pec_base_l, loc=<0 0 0>, rot=<0 0 0 1>}
 deform_pec_aim_base_r {parent=spine_04, loc=<3.2452 9.3309 1.2237>, rot=<-0.4033 -0.0603 0.8507 0.3318>}
 deform_pec_aim_r {parent=deform_pec_aim_base_r, loc=<0 0 0>, rot=<0 0 0 1>}
 deform_pec_base_r {parent=deform_pec_aim_r, loc=<-9.3667 -0.0003 -0.0004>, rot=<0 0 0 1>}
 deform_pec_r {parent=deform_pec_base_r, loc=<0 0 0>, rot=<0 0 0 1>}
 base_purse_l {parent=spine_04, loc=<1.0642 3.0869 -0.6221>, rot=<0.0269 -0.9524 0.0255 -0.3027>}
 dyn_purse_01_l {parent=base_purse_l, loc=<0 0 0>, rot=<0 0 0 -1>}
 dyn_purse_02_l {parent=dyn_purse_01_l, loc=<8.6927 0 0>, rot=<0 0 0 1>}
 dyn_purse_03_l {parent=dyn_purse_02_l, loc=<7.5497 0.0355 0.2217>, rot=<0.0411 -0.0893 0.0184 -0.995>}
 thigh_l {parent=pelvis, loc=<-1.0855 -0.2109 -8.7053>, rot=<0.0009 0.0355 0.0155 -0.9993>}
 calf_l {parent=thigh_l, loc=<-42.5284 -0.0019 -0.7297>, rot=<0.0006 -0.0161 -0.0391 -0.9991>}
 foot_l {parent=calf_l, loc=<-46.7401 -2.5014 -0.592>, rot=<0.0427 -0.0238 0.0353 -0.9982>}
 ball_l {parent=foot_l, loc=<-3.5658 10.8066 -0.1544>, rot=<0.0001 -0.0001 -0.7063 -0.7079>}
 calf_twist_02_l {parent=calf_l, loc=<-19.1546 -1.7943 -0.2638>, rot=<0.0237 0.0095 -0.0035 -0.9997>}
 calf_twist_01_l {parent=calf_l, loc=<-35.8483 -1.9197 -0.5794>, rot=<0.0237 0.0095 -0.0035 -0.9997>}
 deform_knee_in_base_l {parent=calf_l, loc=<0.8031 -5.0231 0.2456>, rot=<0 0 0 -1>}
 deform_knee_in_l {parent=deform_knee_in_base_l, loc=<0 0 0>, rot=<0 0 0 1>}
 deform_knee_out_base_l {parent=calf_l, loc=<0 4 0>, rot=<0 0 0 -1>}
 deform_knee_out_l {parent=deform_knee_out_base_l, loc=<0 0 0>, rot=<0 0 0 1>}
 calf_counter_l {parent=calf_l, loc=<0 0 0>, rot=<0 0 0 1>}
 thigh_twist_01_l {parent=thigh_l, loc=<-27.952 0 0>, rot=<0.0013 0 0 -1>}
 deform_groin_base_l {parent=thigh_l, loc=<0.4801 5.6776 -0.2486>, rot=<0 0 0 -1>}
 deform_groin_l {parent=deform_groin_base_l, loc=<0 0 0>, rot=<0 0 0 1>}
 base_short_fr_l {parent=thigh_l, loc=<-8.2009 6.3182 -1.4354>, rot=<-0.0399 -0.7433 -0.6673 -0.0259>}
 dyn_short_fr_l_01 {parent=base_short_fr_l, loc=<0 0 0>, rot=<0 0 0 1>}
 dyn_short_fr_l_02 {parent=dyn_short_fr_l_01, loc=<3.8331 0 0>, rot=<0 0 0 1>}
 base_short_bk_l {parent=thigh_l, loc=<-8.3348 -9.2325 -1.2504>, rot=<-0.002 -0.7224 0.6915 -0.0013>}
 dyn_short_bk_01_l {parent=base_short_bk_l, loc=<0 0 0>, rot=<0 0 0 1>}
 dyn_short_bk_02_l {parent=dyn_short_bk_01_l, loc=<3.4274 0 0>, rot=<0 0 0 1>}
 base_short_sd_l {parent=thigh_l, loc=<-7.465 -1.588 -6.9974>, rot=<0.0026 -0.9895 -0.0353 -0.1399>}
 dyn_short_sd_01_l {parent=base_short_sd_l, loc=<0 0 0>, rot=<0 0 0 1>}
 dyn_short_sd_02_l {parent=dyn_short_sd_01_l, loc=<3.8331 0 0>, rot=<0 0 0 1>}
 thigh_counter_l {parent=thigh_l, loc=<-42.5284 -0.0019 -0.7297>, rot=<0.0006 -0.0161 -0.0391 -0.9991>}
 thigh_r {parent=pelvis, loc=<-1.0855 -0.2109 8.7053>, rot=<0.0355 -0.0009 0.9993 0.0155>}
 calf_r {parent=thigh_r, loc=<42.5284 0.0019 0.7297>, rot=<0.0006 -0.0161 -0.0391 -0.9991>}
 foot_r {parent=calf_r, loc=<46.7401 2.5014 0.592>, rot=<0.0427 -0.0238 0.0353 -0.9982>}
 ball_r {parent=foot_r, loc=<3.5658 -10.8066 0.1544>, rot=<0.0001 -0.0001 -0.7063 -0.7079>}
 calf_twist_02_r {parent=calf_r, loc=<19.1546 1.7943 0.2638>, rot=<0.0237 0.0095 -0.0035 -0.9997>}
 calf_twist_01_r {parent=calf_r, loc=<35.8483 1.9197 0.5794>, rot=<0.0237 0.0095 -0.0035 -0.9997>}
 deform_knee_in_base_r {parent=calf_r, loc=<-0.8031 5.0231 -0.2457>, rot=<0 0 0 -1>}
 deform_knee_in_r {parent=deform_knee_in_base_r, loc=<0 0 0>, rot=<0 0 0 1>}
 deform_knee_out_base_r {parent=calf_r, loc=<0 -4 0>, rot=<0 0 0 -1>}
 deform_knee_out_r {parent=deform_knee_out_base_r, loc=<0 0 0>, rot=<0 0 0 1>}
 calf_counter_r {parent=calf_r, loc=<0 0 0>, rot=<0 0 0 -1>}
 thigh_twist_01_r {parent=thigh_r, loc=<27.952 0 0>, rot=<0.0013 0 0 -1>}
 deform_groin_base_r {parent=thigh_r, loc=<-0.4801 -5.6776 0.2486>, rot=<0 0 0 -1>}
 deform_groin_r {parent=deform_groin_base_r, loc=<0 0 0>, rot=<0 0 0 1>}
 base_short_sd_r {parent=thigh_r, loc=<7.465 1.5879 6.9974>, rot=<0.9895 0.0026 0.1399 -0.0353>}
 dyn_short_sd_01_r {parent=base_short_sd_r, loc=<0 0 0>, rot=<0 0 0 1>}
 dyn_short_sd_02_r {parent=dyn_short_sd_01_r, loc=<3.833 0 0>, rot=<0 0 0 1>}
 base_short_bk_r {parent=thigh_r, loc=<8.3348 9.2326 1.2505>, rot=<-0.7224 0.002 -0.0013 -0.6915>}
 dyn_short_bk_01_r {parent=base_short_bk_r, loc=<0 0 0>, rot=<0 0 0 1>}
 dyn_short_bk_02_r {parent=dyn_short_bk_01_r, loc=<3.4274 0 0>, rot=<0 0 0 1>}
 base_short_fr_r {parent=thigh_r, loc=<8.2009 -6.3182 1.4354>, rot=<0.7433 -0.0399 0.0259 -0.6673>}
 dyn_short_fr_r_01 {parent=base_short_fr_r, loc=<0 0 0>, rot=<0 0 0 1>}
 dyn_short_fr_r_02 {parent=dyn_short_fr_r_01, loc=<3.833 -0.0001 0>, rot=<0 0 0 1>}
 thigh_counter_r {parent=thigh_r, loc=<42.5284 0.0019 0.7298>, rot=<0.0006 -0.0161 -0.0391 -0.9991>}
 deform_glute_base_l {parent=pelvis, loc=<-7.3466 -9.0213 -9.1245>, rot=<0.0172 0.031 -0.4477 -0.8935>}
 deform_glute_l {parent=deform_glute_base_l, loc=<0 0 0>, rot=<0 0 0 1>}
 deform_glute_base_r {parent=pelvis, loc=<-7.3464 -9.0213 9.1245>, rot=<0.031 -0.0172 0.8935 -0.4477>}
 deform_glute_r {parent=deform_glute_base_r, loc=<0 0 0>, rot=<0 0 0 1>}
 base_pendant_bk_r {parent=pelvis, loc=<6.7369 -6.3054 10.1891>, rot=<0.2258 -0.2897 0.9189 -0.144>}
 dyn_pendant_bk_01_r {parent=base_pendant_bk_r, loc=<0 0 0>, rot=<0 0 0 1>}
 dyn_pendant_bk_02_r {parent=dyn_pendant_bk_01_r, loc=<5.3168 0 0>, rot=<0 0 0 1>}
]
variables [
 ArmBackInMelee {type=double, default=0}
 Crouching {type=bool, default=False}
 Melee {type=bool, default=False}
 OnGrinRail {type=bool, default=False}
 RBEnableAndGrindRail {type=bool, default=False}
 RailSeated {type=double, default=0}
 SetTranslate {type=FVector, default=(X=0,Y=0,Z=0)}
 SkyDiveMode {type=bool, default=False}
 Skydiving {type=bool, default=False}
 TacticalSprintingJump {type=double, default=0}
 bandolier constraint alpha {type=double, default=0}
 isParachute {type=bool, default=False}
]
