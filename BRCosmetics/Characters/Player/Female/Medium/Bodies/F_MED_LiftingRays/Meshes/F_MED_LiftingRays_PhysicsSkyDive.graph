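# Physics asset for the LiftingRays female medium body, a skydive variant covering its bandolier, purse and pendant dynamic bones.
# Capsules on pelvis, spine and thighs form body volumes, linked by free constraints.
# Front and back bandolier chains run from a root on spine_04 to the purse, with locked twist and mostly single-axis limited swing.
# Pendant chains hang from the back purse bone and pelvis; extra free constraints cross-link purse and strap bones.
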PhysicsAsset /BRCosmetics/Characters/Player/Female/Medium/Bodies/F_MED_LiftingRays/Meshes/F_MED_LiftingRays_PhysicsSkyDive
bodies {
 base_bandolier_root [{shapeType=Capsule, length=0.505, radius=0.505, shapeName=base_bandolier_root_capsule, rotation=<rot 0 -90 0>}]
 dyn_bandolier_bk_02_l [{shapeType=Capsule, length=8, radius=1, shapeName=dyn_bandolier_bk_02_l_capsule, center=<3.6359 -0.6961 0>, rotation=<rot 90 180 174.8996>}]
 dyn_bandolier_bk_03_l [{shapeType=Capsule, length=5.5, radius=1, shapeName=dyn_bandolier_bk_03_l_capsule, center=<2.3463 -0.3643 0>, rotation=<rot 90 0 0>}]
 dyn_bandolier_bk_04_l [{shapeType=Capsule, length=5, radius=1, shapeName=dyn_bandolier_bk_04_l_capsule, center=<2.9133 -0.3014 0>, rotation=<rot 90 0 0>}]
 dyn_bandolier_bk_05_l [{shapeType=Capsule, length=3, radius=1, shapeName=dyn_bandolier_bk_05_l_capsule, center=<1.4795 -0.4949 0>, rotation=<rot 90 0 0>}]
 dyn_bandolier_fr_01_l [{shapeType=Capsule, length=6, radius=1, shapeName=dyn_bandolier_fr_01_l_capsule, center=<1.9683 0.7402 0>, rotation=<rot 90 0 -1.032>}]
 dyn_bandolier_fr_02_l [{
   shapeType=Capsule
   length=6.5
   radius=1
   shapeName=dyn_bandolier_fr_02_l_capsule
   center=<1.8976 1.0577 0.3706>
   rotation=<rot 72.0259 -84.0618 -90>
  }]
 dyn_bandolier_fr_03_l [{shapeType=Capsule, length=5.5, radius=1, shapeName=dyn_bandolier_fr_03_l_capsule, center=<2.4575 0.832 0>, rotation=<rot 90 180 177.8665>}]
 dyn_bandolier_fr_04_l [{
   shapeType=Capsule
   length=6
   radius=1
   shapeName=dyn_bandolier_fr_04_l_capsule
   center=<2.5219 0.7572 0>
   rotation=<rot 90 179.821 -177.9469>
  }]
 dyn_bandolier_fr_05_l [{shapeType=Capsule, length=3, radius=1, shapeName=dyn_bandolier_fr_05_l_capsule, center=<1.479 0 0>, rotation=<rot 90 0 0>}]
 dyn_pendant_01_l [{
   shapeType=Capsule
   length=2.005
   radius=0.505
   shapeName=dyn_pendant_01_l_capsule
   center=<0.8649 -0.0806 0.1415>
   rotation=<rot 0 93.9497 -83.0353>
  }]
 dyn_pendant_02_l [{
   shapeType=Capsule
   length=4.1398
   radius=2.2038
   shapeName=dyn_pendant_02_l_capsule
   center=<4.7249 0.14 0.648>
   rotation=<rot 0 -89.8705 -89.8661>
  }]
 dyn_pendant_bk_01_r [{shapeType=Capsule, length=5, radius=1, shapeName=dyn_pendant_bk_01_r_capsule, center=<4.3072 0 0.5207>, rotation=<rot 90 0 0>}]
 dyn_pendant_bk_02_r [{shapeType=Capsule, length=5, radius=1, shapeName=dyn_pendant_bk_02_r_capsule, center=<4.3072 0 0.4667>, rotation=<rot 90 0 0>}]
 dyn_purse_03_l [{shapeType=Capsule, length=0.505, radius=0.505, shapeName=dyn_purse_03_l_capsule, center=<2 0 0>, rotation=<rot 0 -90 0>}]
 dyn_purse_bk_l [
  {
   shapeType=Capsule
   length=11.3552
   radius=1.6164
   shapeName=dyn_purse_bk_l_capsule
   center=<5.6473 0.4713 0.3972>
   rotation=<rot -0.003 -89.5036 -89.402>
  }
  {
   shapeType=Capsule
   length=11.3552
   radius=1.6164
   shapeName=dyn_purse_bk_l_capsule
   center=<5.5011 5.0149 0.3046>
   rotation=<rot 90 -0.3205 0>
  }
 ]
 dyn_purse_fr_l [{
   shapeType=Capsule
   length=10.3977
   radius=1.879
   shapeName=dyn_purse_fr_l_capsule
   center=<5.5341 -0.6437 0.4981>
   rotation=<rot -0.0016 -90.3613 -90.3043>
  }]
 pelvis [{
   shapeType=Capsule
   length=7.6714
   radius=10.5
   shapeName=pelvis_capsule
   center=<-0.5083 -2.4894 -0.0014>
   rotation=<rot -0.1264 -90 -0.0924>
  }]
 spine_02 [{shapeType=Capsule, length=9.5, radius=8.9, shapeName=spine_02_capsule, center=<6.2656 1.456 -0.3377>, rotation=<rot 90 0 0>}]
 spine_03 [
  {shapeType=Capsule, length=5.005, radius=8.5, shapeName=spine_03_capsule, center=<0 1.484 0>}
  {shapeType=Capsule, length=8, radius=9, shapeName=spine_03_capsule, center=<7.5214 1.5203 0>}
  {shapeType=Capsule, length=5.5, radius=8, shapeName=spine_03_capsule, center=<10.1406 0.9481 0>, rotation=<rot 90 0 0>}
  {shapeType=Capsule, length=4, radius=4, shapeName=spine_03_capsule, center=<4.6419 -2.6964 -6.0977>, rotation=<rot 90 -11.972 0>}
 ]
 spine_04 [
  {shapeType=Capsule, length=7.8044, radius=7.985, shapeName=spine_04_capsule, center=<2.3597 3.2377 0>, rotation=<rot 0.0002 -90 -0.0001>}
  {shapeType=Capsule, length=11, radius=7, shapeName=spine_04_capsule, center=<9.8393 -0.9086 1.3174>}
  {shapeType=Capsule, length=10.5, radius=9.5, shapeName=spine_04_capsule, center=<3.3676 1.3592 0>}
 ]
 thigh_l [{
   shapeType=Capsule
   length=29.1231
   radius=8.1851
   shapeName=thigh_l_capsule
   center=<-17.6334 -1.016 1.5172>
   rotation=<rot 0.2657 -90.257 -87.8959>
  }]
 thigh_r [{
   shapeType=Capsule
   length=29.1231
   radius=8.1851
   shapeName=thigh_r_capsule
   center=<17.6334 1.016 -1.5172>
   rotation=<rot 0.2657 -90.257 92.1041>
  }]
}
constraints [
 {
  bone1Name=spine_02
  bone2Name=pelvis
  swing1LimitDegrees=45
  swing1Motion=Free
  swing2LimitDegrees=45
  swing2Motion=Free
  twistLimitDegrees=45
  twistMotion=Free
 }
 {
  bone1Name=spine_03
  bone2Name=spine_02
  swing1LimitDegrees=45
  swing1Motion=Free
  swing2LimitDegrees=45
  swing2Motion=Free
  twistLimitDegrees=45
  twistMotion=Free
 }
 {
  bone1Name=thigh_l
  bone2Name=pelvis
  swing1LimitDegrees=45
  swing1Motion=Free
  swing2LimitDegrees=45
  swing2Motion=Free
  twistLimitDegrees=45
  twistMotion=Free
 }
 {
  bone1Name=thigh_r
  bone2Name=pelvis
  swing1LimitDegrees=45
  swing1Motion=Free
  swing2LimitDegrees=45
  swing2Motion=Free
  twistLimitDegrees=45
  twistMotion=Free
 }
 {
  bone1Name=spine_04
  bone2Name=spine_03
  swing1LimitDegrees=45
  swing1Motion=Free
  swing2LimitDegrees=45
  swing2Motion=Free
  twistLimitDegrees=45
  twistMotion=Free
 }
 {
  bone1Name=dyn_purse_bk_l
  bone2Name=dyn_bandolier_bk_05_l
  swing1LimitDegrees=45
  swing1Motion=Locked
  swing2LimitDegrees=20
  twistLimitDegrees=45
  twistMotion=Locked
 }
 {
  bone1Name=dyn_purse_fr_l
  bone2Name=dyn_bandolier_fr_05_l
  swing1LimitDegrees=45
  swing1Motion=Locked
  swing2LimitDegrees=3
  twistLimitDegrees=45
  twistMotion=Locked
 }
 {
  bone1Name=dyn_purse_03_l
  bone2Name=spine_04
  swing1LimitDegrees=45
  swing1Motion=Free
  swing2LimitDegrees=45
  swing2Motion=Free
  twistLimitDegrees=45
  twistMotion=Free
 }
 {
  bone1Name=dyn_bandolier_bk_02_l
  bone2Name=base_bandolier_root
  swing1LimitDegrees=8
  swing2LimitDegrees=3
  twistLimitDegrees=45
  twistMotion=Locked
 }
 {
  bone1Name=dyn_bandolier_bk_03_l
  bone2Name=dyn_bandolier_bk_02_l
  swing1LimitDegrees=3
  swing2LimitDegrees=7
  twistLimitDegrees=45
  twistMotion=Locked
 }
 {
  bone1Name=dyn_bandolier_bk_04_l
  bone2Name=dyn_bandolier_bk_03_l
  swing1LimitDegrees=3
  swing1Motion=Locked
  swing2LimitDegrees=10
  twistLimitDegrees=45
  twistMotion=Locked
 }
 {
  bone1Name=dyn_bandolier_bk_05_l
  bone2Name=dyn_bandolier_bk_04_l
  swing1LimitDegrees=3
  swing1Motion=Locked
  swing2LimitDegrees=20
  twistLimitDegrees=45
  twistMotion=Locked
 }
 {
  bone1Name=dyn_bandolier_fr_02_l
  bone2Name=dyn_bandolier_fr_01_l
  swing1LimitDegrees=5
  swing1Motion=Locked
  swing2LimitDegrees=15
  twistLimitDegrees=45
  twistMotion=Locked
 }
 {
  bone1Name=dyn_bandolier_fr_01_l
  bone2Name=base_bandolier_root
  swing1LimitDegrees=5
  swing1Motion=Locked
  swing2LimitDegrees=5
  twistLimitDegrees=45
  twistMotion=Locked
 }
 {
  bone1Name=dyn_bandolier_fr_03_l
  bone2Name=dyn_bandolier_fr_02_l
  swing1LimitDegrees=5
  swing1Motion=Locked
  swing2LimitDegrees=15
  twistLimitDegrees=45
  twistMotion=Locked
 }
 {
  bone1Name=dyn_bandolier_fr_05_l
  bone2Name=dyn_bandolier_fr_04_l
  swing1LimitDegrees=3
  swing1Motion=Locked
  swing2LimitDegrees=20
  twistLimitDegrees=45
  twistMotion=Locked
 }
 {
  bone1Name=dyn_bandolier_fr_04_l
  bone2Name=dyn_bandolier_fr_03_l
  swing1LimitDegrees=3
  swing2LimitDegrees=15
  twistLimitDegrees=45
  twistMotion=Locked
 }
 {
  bone1Name=base_bandolier_root
  bone2Name=spine_04
  swing1LimitDegrees=45
  swing1Motion=Free
  swing2LimitDegrees=45
  swing2Motion=Free
  twistLimitDegrees=45
  twistMotion=Free
 }
 {
  bone1Name=dyn_purse_fr_l
  bone2Name=dyn_bandolier_bk_05_l
  swing1LimitDegrees=45
  swing1Motion=Free
  swing2LimitDegrees=45
  swing2Motion=Free
  twistLimitDegrees=45
  twistMotion=Free
 }
 {
  bone1Name=dyn_bandolier_fr_04_l
  bone2Name=dyn_purse_03_l
  swing1LimitDegrees=45
  swing1Motion=Free
  swing2LimitDegrees=45
  swing2Motion=Free
  twistLimitDegrees=45
  twistMotion=Free
 }
 {
  bone1Name=dyn_bandolier_bk_04_l
  bone2Name=dyn_purse_03_l
  swing1LimitDegrees=45
  swing1Motion=Free
  swing2LimitDegrees=45
  swing2Motion=Free
  twistLimitDegrees=45
  twistMotion=Free
 }
 {bone1Name=dyn_pendant_01_l, bone2Name=dyn_purse_bk_l, swing1LimitDegrees=20, swing2LimitDegrees=20, twistLimitDegrees=45, twistMotion=Locked}
 {
  bone1Name=dyn_pendant_02_l
  bone2Name=dyn_pendant_01_l
  swing1LimitDegrees=10
  swing2LimitDegrees=10
  twistLimitDegrees=45
  twistMotion=Locked
 }
 {bone1Name=dyn_pendant_bk_01_r, bone2Name=pelvis, swing1LimitDegrees=30, swing2LimitDegrees=30, twistLimitDegrees=45, twistMotion=Locked}
 {
  bone1Name=dyn_pendant_bk_02_r
  bone2Name=dyn_pendant_bk_01_r
  swing1LimitDegrees=30
  swing2LimitDegrees=40
  twistLimitDegrees=45
  twistMotion=Locked
 }
]
