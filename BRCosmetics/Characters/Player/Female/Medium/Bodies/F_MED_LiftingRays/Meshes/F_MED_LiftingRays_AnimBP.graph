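# Animation Blueprint for the F_MED_LiftingRays character body, adding secondary motion to shorts, purse, pendant and bandolier over the parent mesh's pose.
# A Control Rig feeds cached AnimDynamics chains, then skydive or default RigidBody simulation, bypassed for a bandolier offset on grind rails.
# Fixed bone tweaks finish the pose.
# The event graph computes gravity override and Control Rig inputs from movement, weapon and spine-rotation checks.
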AnimBlueprint /BRCosmetics/Characters/Player/Female/Medium/Bodies/F_MED_LiftingRays/Meshes/F_MED_LiftingRays_AnimBP
target_skeleton=&/F_MED_LiftingRays_Skeleton
graphs [
 AnimGraph {nodes [
   n1 = RigidBody!(
    ComponentPose=Local To Component(LocalPose=Use cached pose 'AnimDynamicsCache'())
    Alpha
    OverrideWorldGravity:Vector=Get Gravity Override()
    ExternalForce
    SimSpaceSettings:Sim Space Settings Structure=(WorldAlpha=0.5,VelocityScaleZ=0.5,DampingAlpha=1,MaxLinearVelocity=500,MaxAngularVelocity=10,MaxLinearAcceleration=300,MaxAngularAcceleration=500,ExternalLinearDragV=(X=0,Y=0,Z=0),ExternalLinearVelocity=(X=0,Y=0,Z=0),ExternalAngularVelocity=(X=0,Y=0,Z=0))
   ) -> (Pose) {
    bOverrideWorldGravity=true
    overridePhysicsAsset=&/F_MED_LiftingRays_PhysicsSkyDive
    simulationSpace=BaseBoneSpace
    alphaBoolBlend {blendOption=Linear}
    baseBoneRef {boneName=spine_05}
    componentAppliedLinearAccClamp=<3000 3000 3000>
    simSpaceSettings {maxAngularAcceleration=500, maxAngularVelocity=10, maxLinearAcceleration=300, maxLinearVelocity=500, velocityScaleZ=0.5, worldAlpha=0.5}
   }
   n2 = RigidBody(
    ComponentPose=Local To Component(LocalPose=Use cached pose 'AnimDynamicsCache'())
    OverrideWorldGravity=Get Gravity Override()
    SimSpaceSettings=(WorldAlpha=0.5,VelocityScaleZ=0.5,DampingAlpha=1,MaxLinearVelocity=500,MaxAngularVelocity=10,MaxLinearAcceleration=300,MaxAngularAcceleration=500,ExternalLinearDragV=(X=0,Y=0,Z=0),ExternalLinearVelocity=(X=0,Y=0,Z=0),ExternalAngularVelocity=(X=0,Y=0,Z=0))
   ) {
    bOverrideWorldGravity=true
    overridePhysicsAsset=&/F_MED_LiftingRays_Physics
    simulationSpace=BaseBoneSpace
    alphaBoolBlend {blendOption=Linear}
    alphaScaleBiasClamp {bInterpResult=true, interpSpeedDecreasing=5, interpSpeedIncreasing=5}
    baseBoneRef {boneName=spine_02}
    componentAppliedLinearAccClamp=<3000 3000 3000>
    overrideWorldGravity {z=-980}
    simSpaceSettings {maxAngularAcceleration=500, maxAngularVelocity=10, maxLinearAcceleration=300, maxLinearVelocity=500, velocityScaleZ=0.5, worldAlpha=0.5}
   }
   n3 = "Transform (Modify) Bone - Bone: dyn_purse_bk_l"(ComponentPose=@n2, Alpha=1, Translation="0,0,0", Rotation="40,0,0", Scale="1,1,1") {
    rotationMode=BMM_Additive
    rotationSpace=BCS_BoneSpace
    translationMode=BMM_Additive
    translationSpace=BCS_BoneSpace
    alphaBoolBlend {blendInTime=0.2, blendOption=Linear, blendOutTime=0.2}
    alphaScaleBiasClamp {bInterpResult=true, interpSpeedDecreasing=7, interpSpeedIncreasing=2.5}
    boneToModify {boneName=dyn_purse_bk_l}
    rotation {pitch=40}
   }
   n4 = "Transform (Modify) Bone - Bone: dyn_bandolier_bk_01_extra_l"(ComponentPose=@n3, Alpha=1, Translation="0,-5,0", Rotation="0,0,0", Scale="1,1,1") {
    rotationMode=BMM_Additive
    rotationSpace=BCS_BoneSpace
    translationMode=BMM_Additive
    translationSpace=BCS_BoneSpace
    alphaBoolBlend {blendInTime=0.2, blendOption=Linear, blendOutTime=0.2}
    alphaScaleBiasClamp {bInterpResult=true, interpSpeedDecreasing=3}
    boneToModify {boneName=dyn_bandolier_bk_01_extra_l}
    translation {y=-5}
   }
   n5 = "Transform (Modify) Bone - Bone: dyn_bandolier_bk_02_extra_l"(ComponentPose=@n4, Alpha=1, Translation="0,-2,0", Rotation="0,0,0", Scale="1,1,1") {
    rotationMode=BMM_Additive
    rotationSpace=BCS_BoneSpace
    translationMode=BMM_Additive
    translationSpace=BCS_BoneSpace
    alphaBoolBlend {blendInTime=0.2, blendOption=Linear, blendOutTime=0.2}
    alphaScaleBiasClamp {bInterpResult=true, interpSpeedDecreasing=3}
    boneToModify {boneName=dyn_bandolier_bk_02_extra_l}
    translation {y=-2}
   }
   n6 = Blend Poses by bool(BlendPose_0=Component To Local(ComponentPose=@n1), BlendPose_1=Component To Local(ComponentPose=@n5), bActiveValue=Get bIsSkyDiving()) {blendType=Linear}
   n7 = "Transform (Modify) Bone - Bone: base_bandolier_root"!(
    ComponentPose=Local To Component(LocalPose=Use cached pose 'AnimDynamicsCache'())
    bAlphaBoolEnabled:bool=True
    Translation="0,-1.5,0"
    Rotation="0.026,3.296,-0.030"
    Scale="1,1,1"
   ) -> (Pose) {
    alphaInputType=Bool
    rotationMode=BMM_Additive
    rotationSpace=BCS_BoneSpace
    translationMode=BMM_Additive
    translationSpace=BCS_BoneSpace
    alphaBoolBlend {blendInTime=0.2, blendOption=Linear, blendOutTime=0.2}
    boneToModify {boneName=base_bandolier_root}
    rotation=<rot 0.026 3.296 -0.03>
    translation {y=-1.5}
   }
   n8 = Component To Local(ComponentPose=@n7)
   n9 = Get bIsOnGrindRail()
   n10 = "Transform (Modify) Bone - Bone: thumb_01_r"!(
    ComponentPose=Local To Component(LocalPose=Blend Poses by bool(BlendPose_0=@n8, BlendPose_1=@n6, bActiveValue=@n9) {blendType=Linear})
    bAlphaBoolEnabled:bool=True
    Translation="0,0,0"
    Rotation="0,6,0"
    Scale="1,1,1"
   ) -> (Pose) {
    alphaInputType=Bool
    rotationMode=BMM_Additive
    rotationSpace=BCS_BoneSpace
    translationMode=BMM_Additive
    translationSpace=BCS_BoneSpace
    alphaBoolBlend {blendOption=Linear}
    boneToModify {boneName=thumb_01_r}
    rotation {yaw=6}
   }
   n11 = "Transform (Modify) Bone - Bone: middle_02_r"!(ComponentPose=@n10, bAlphaBoolEnabled:bool=True, Translation="0,0,0", Rotation="5,13,0", Scale="1,1,1") -> (Pose) {
    alphaInputType=Bool
    rotationMode=BMM_Additive
    rotationSpace=BCS_BoneSpace
    translationMode=BMM_Additive
    translationSpace=BCS_BoneSpace
    alphaBoolBlend {blendOption=Linear}
    boneToModify {boneName=middle_02_r}
    rotation {pitch=5, yaw=13}
   }
   n12 = "Transform (Modify) Bone - Bone: ring_02_r"!(ComponentPose=@n11, bAlphaBoolEnabled:bool=True, Translation="0,0,0", Rotation="6,7,0", Scale="1,1,1") -> (Pose) {
    alphaInputType=Bool
    rotationMode=BMM_Additive
    rotationSpace=BCS_BoneSpace
    translationMode=BMM_Additive
    translationSpace=BCS_BoneSpace
    alphaBoolBlend {blendOption=Linear}
    boneToModify {boneName=ring_02_r}
    rotation {pitch=6, yaw=7}
   }
   n13 = "Transform (Modify) Bone - Bone: dyn_bandolier_fr_01_l"!(ComponentPose=@n12, bAlphaBoolEnabled:bool=True, Translation="0,0.2,0", Rotation="0,0,0", Scale="1,1,1") -> (Pose) {
    alphaInputType=Bool
    rotationMode=BMM_Additive
    rotationSpace=BCS_BoneSpace
    translationMode=BMM_Additive
    translationSpace=BCS_BoneSpace
    alphaBoolBlend {blendOption=Linear}
    boneToModify {boneName=dyn_bandolier_fr_01_l}
    translation {y=0.2}
   }
   n14 = "Transform (Modify) Bone - Bone: dyn_bandolier_fr_03_l"!(ComponentPose=@n13, bAlphaBoolEnabled:bool=True, Translation="0,0,0", Rotation="0.000,-10.466,-0.000", Scale="1,1,1") -> (Pose) {
    alphaInputType=Bool
    rotationMode=BMM_Additive
    rotationSpace=BCS_BoneSpace
    translationMode=BMM_Additive
    translationSpace=BCS_BoneSpace
    alphaBoolBlend {blendOption=Linear}
    boneToModify {boneName=dyn_bandolier_fr_03_l}
    rotation=<rot 0 -10.466 0>
   }
   Output Pose(Result=Component To Local(ComponentPose=@n14)) {layerGroup=None}
   n15 = Anim Dynamics - Bone: dyn_short_sd_01_r(ComponentPose=Local To Component(LocalPose=Use cached pose 'CtrlRigCache'()), Alpha=1, ExternalForce="0,0,0") {
    angularDampingOverride=1
    angularSpringConstant=25000
    bAngularSpring=true
    bOverrideAngularDamping=true
    bUseGravityOverride=true
    gravityScale=0.35
    simulationSpace=BoneRelative
    alphaBoolBlend {blendOption=Linear}
    boundBone {boneName=dyn_short_sd_01_r}
    componentAppliedLinearAccClamp=<2000 2000 2000>
    componentLinearAccScale=<0.6 0.6 0.6>
    componentLinearVelScale=<0.6 0.6 0.6>
    gravityOverride {z=-980}
    physicsBodyDefinitions [{
      boundBone {boneName=dyn_short_sd_01_r}
      boxExtents=<20 30 30>
      constraintSetup {angularLimitsMax=<0 25 15>, angularLimitsMin=<0 -15 -15>, angularTarget=<1 0 0>}
      localJointOffset=<6 0 0>
     }]
    relativeSpaceBone {boneName=pelvis}
   }
   n16 = Anim Dynamics - Bone: dyn_short_sd_02_r(ComponentPose=@n15, Alpha=1, ExternalForce="0,0,0") {
    angularDampingOverride=1
    angularSpringConstant=25000
    bAngularSpring=true
    bOverrideAngularDamping=true
    bUseGravityOverride=true
    gravityScale=0.35
    simulationSpace=BoneRelative
    alphaBoolBlend {blendOption=Linear}
    boundBone {boneName=dyn_short_sd_02_r}
    componentAppliedLinearAccClamp=<2000 2000 2000>
    componentLinearAccScale=<0.6 0.6 0.6>
    componentLinearVelScale=<0.6 0.6 0.6>
    gravityOverride {z=-980}
    physicsBodyDefinitions [{
      boundBone {boneName=dyn_short_sd_02_r}
      boxExtents=<20 30 30>
      constraintSetup {angularLimitsMax=<0 25 15>, angularLimitsMin=<0 -15 -15>, angularTarget=<1 0 0>}
      localJointOffset=<6 0 0>
     }]
    relativeSpaceBone {boneName=pelvis}
   }
   n17 = Anim Dynamics - Bone: dyn_short_bk_01_r(ComponentPose=@n16, Alpha=1, ExternalForce="0,0,0") {
    angularDampingOverride=1
    angularSpringConstant=25000
    bAngularSpring=true
    bOverrideAngularDamping=true
    bUseGravityOverride=true
    gravityScale=0.35
    simulationSpace=BoneRelative
    alphaBoolBlend {blendOption=Linear}
    boundBone {boneName=dyn_short_bk_01_r}
    componentAppliedLinearAccClamp=<2000 2000 2000>
    componentLinearAccScale=<0.6 0.6 0.6>
    componentLinearVelScale=<0.6 0.6 0.6>
    gravityOverride {z=-980}
    physicsBodyDefinitions [{
      boundBone {boneName=dyn_short_bk_01_r}
      boxExtents=<20 30 30>
      constraintSetup {angularLimitsMax=<0 25 15>, angularLimitsMin=<0 -15 -15>, angularTarget=<1 0 0>}
      localJointOffset=<6 0 0>
     }]
    relativeSpaceBone {boneName=pelvis}
   }
   n18 = Anim Dynamics - Bone: dyn_short_bk_02_r(ComponentPose=@n17, Alpha=1, ExternalForce="0,0,0") {
    angularDampingOverride=1
    angularSpringConstant=25000
    bAngularSpring=true
    bOverrideAngularDamping=true
    bUseGravityOverride=true
    gravityScale=0.35
    simulationSpace=BoneRelative
    alphaBoolBlend {blendOption=Linear}
    boundBone {boneName=dyn_short_bk_02_r}
    componentAppliedLinearAccClamp=<2000 2000 2000>
    componentLinearAccScale=<0.6 0.6 0.6>
    componentLinearVelScale=<0.6 0.6 0.6>
    gravityOverride {z=-980}
    physicsBodyDefinitions [{
      boundBone {boneName=dyn_short_bk_02_r}
      boxExtents=<20 30 30>
      constraintSetup {angularLimitsMax=<0 25 15>, angularLimitsMin=<0 -15 -15>, angularTarget=<1 0 0>}
      localJointOffset=<6 0 0>
     }]
    relativeSpaceBone {boneName=pelvis}
   }
   n19 = Anim Dynamics - Bone: dyn_short_fr_r_01(ComponentPose=@n18, Alpha=1, ExternalForce="0,0,0") {
    angularDampingOverride=1
    angularSpringConstant=25000
    bAngularSpring=true
    bOverrideAngularDamping=true
    bUseGravityOverride=true
    gravityScale=0.35
    simulationSpace=BoneRelative
    alphaBoolBlend {blendOption=Linear}
    boundBone {boneName=dyn_short_fr_r_01}
    componentAppliedLinearAccClamp=<2000 2000 2000>
    componentLinearAccScale=<0.6 0.6 0.6>
    componentLinearVelScale=<0.6 0.6 0.6>
    gravityOverride {z=-980}
    physicsBodyDefinitions [{
      boundBone {boneName=dyn_short_fr_r_01}
      boxExtents=<20 30 30>
      constraintSetup {angularLimitsMax=<0 25 15>, angularLimitsMin=<0 0 -15>, angularTarget=<1 0 0>}
      localJointOffset=<6 0 0>
     }]
    relativeSpaceBone {boneName=pelvis}
   }
   n20 = Anim Dynamics - Bone: dyn_short_fr_r_02(ComponentPose=@n19, Alpha=1, ExternalForce="0,0,0") {
    angularDampingOverride=1
    angularSpringConstant=25000
    bAngularSpring=true
    bOverrideAngularDamping=true
    bUseGravityOverride=true
    gravityScale=0.35
    simulationSpace=BoneRelative
    alphaBoolBlend {blendOption=Linear}
    boundBone {boneName=dyn_short_fr_r_02}
    componentAppliedLinearAccClamp=<2000 2000 2000>
    componentLinearAccScale=<0.6 0.6 0.6>
    componentLinearVelScale=<0.6 0.6 0.6>
    gravityOverride {z=-980}
    physicsBodyDefinitions [{
      boundBone {boneName=dyn_short_fr_r_02}
      boxExtents=<20 30 30>
      constraintSetup {angularLimitsMax=<0 25 15>, angularLimitsMin=<0 0 -15>, angularTarget=<1 0 0>}
      localJointOffset=<6 0 0>
     }]
    relativeSpaceBone {boneName=pelvis}
   }
   n21 = Anim Dynamics - Bone: dyn_short_sd_02_l(ComponentPose=@n20, Alpha=1, ExternalForce="0,0,0") {
    angularDampingOverride=1
    angularSpringConstant=25000
    bAngularSpring=true
    bOverrideAngularDamping=true
    bUseGravityOverride=true
    gravityScale=0.35
    simulationSpace=BoneRelative
    alphaBoolBlend {blendOption=Linear}
    boundBone {boneName=dyn_short_sd_02_l}
    componentAppliedLinearAccClamp=<2000 2000 2000>
    componentLinearAccScale=<0.6 0.6 0.6>
    componentLinearVelScale=<0.6 0.6 0.6>
    gravityOverride {z=-980}
    physicsBodyDefinitions [{
      boundBone {boneName=dyn_short_sd_02_l}
      boxExtents=<20 30 30>
      constraintSetup {angularLimitsMax=<0 15 15>, angularLimitsMin=<0 -25 -15>, angularTarget=<1 0 0>}
      localJointOffset=<6 0 0>
     }]
    relativeSpaceBone {boneName=pelvis}
   }
   n22 = Anim Dynamics - Bone: dyn_short_sd_01_l(ComponentPose=@n21, Alpha=1, ExternalForce="0,0,0") {
    angularDampingOverride=1
    angularSpringConstant=25000
    bAngularSpring=true
    bOverrideAngularDamping=true
    bUseGravityOverride=true
    gravityScale=0.35
    simulationSpace=BoneRelative
    alphaBoolBlend {blendOption=Linear}
    boundBone {boneName=dyn_short_sd_01_l}
    componentAppliedLinearAccClamp=<2000 2000 2000>
    componentLinearAccScale=<0.6 0.6 0.6>
    componentLinearVelScale=<0.6 0.6 0.6>
    gravityOverride {z=-980}
    physicsBodyDefinitions [{
      boundBone {boneName=dyn_short_sd_01_l}
      boxExtents=<20 30 30>
      constraintSetup {angularLimitsMax=<0 15 15>, angularLimitsMin=<0 -25 -15>, angularTarget=<1 0 0>}
      localJointOffset=<6 0 0>
     }]
    relativeSpaceBone {boneName=pelvis}
   }
   n23 = Anim Dynamics - Bone: dyn_short_bk_02_l(ComponentPose=@n22, Alpha=1, ExternalForce="0,0,0") {
    angularDampingOverride=1
    angularSpringConstant=25000
    bAngularSpring=true
    bOverrideAngularDamping=true
    bUseGravityOverride=true
    gravityScale=0.35
    simulationSpace=BoneRelative
    alphaBoolBlend {blendOption=Linear}
    boundBone {boneName=dyn_short_bk_02_l}
    componentAppliedLinearAccClamp=<2000 2000 2000>
    componentLinearAccScale=<0.6 0.6 0.6>
    componentLinearVelScale=<0.6 0.6 0.6>
    gravityOverride {z=-980}
    physicsBodyDefinitions [{
      boundBone {boneName=dyn_short_bk_02_l}
      boxExtents=<20 30 30>
      constraintSetup {angularLimitsMax=<0 15 15>, angularLimitsMin=<0 -25 -15>, angularTarget=<1 0 0>}
      localJointOffset=<6 0 0>
     }]
    relativeSpaceBone {boneName=pelvis}
   }
   n24 = Anim Dynamics - Bone: dyn_short_bk_01_l(ComponentPose=@n23, Alpha=1, ExternalForce="0,0,0") {
    angularDampingOverride=1
    angularSpringConstant=25000
    bAngularSpring=true
    bOverrideAngularDamping=true
    bUseGravityOverride=true
    gravityScale=0.35
    simulationSpace=BoneRelative
    alphaBoolBlend {blendOption=Linear}
    boundBone {boneName=dyn_short_bk_01_l}
    componentAppliedLinearAccClamp=<2000 2000 2000>
    componentLinearAccScale=<0.6 0.6 0.6>
    componentLinearVelScale=<0.6 0.6 0.6>
    gravityOverride {z=-980}
    physicsBodyDefinitions [{
      boundBone {boneName=dyn_short_bk_01_l}
      boxExtents=<20 30 30>
      constraintSetup {angularLimitsMax=<0 15 15>, angularLimitsMin=<0 -25 -15>, angularTarget=<1 0 0>}
      localJointOffset=<6 0 0>
     }]
    relativeSpaceBone {boneName=pelvis}
   }
   n25 = Anim Dynamics - Bone: dyn_short_fr_l_02(ComponentPose=@n24, Alpha=1, ExternalForce="0,0,0") {
    angularDampingOverride=1
    angularSpringConstant=25000
    bAngularSpring=true
    bOverrideAngularDamping=true
    bUseGravityOverride=true
    gravityScale=0.35
    simulationSpace=BoneRelative
    alphaBoolBlend {blendOption=Linear}
    boundBone {boneName=dyn_short_fr_l_02}
    componentAppliedLinearAccClamp=<2000 2000 2000>
    componentLinearAccScale=<0.6 0.6 0.6>
    componentLinearVelScale=<0.6 0.6 0.6>
    gravityOverride {z=-980}
    physicsBodyDefinitions [{
      boundBone {boneName=dyn_short_fr_l_02}
      boxExtents=<20 30 30>
      constraintSetup {angularLimitsMax=<0 0 15>, angularLimitsMin=<0 -25 -15>, angularTarget=<1 0 0>}
      localJointOffset=<6 0 0>
     }]
    relativeSpaceBone {boneName=pelvis}
   }
   n26 = Anim Dynamics - Bone: dyn_short_fr_l_01(ComponentPose=@n25, Alpha=1, ExternalForce="0,0,0") {
    angularDampingOverride=1
    angularSpringConstant=25000
    bAngularSpring=true
    bOverrideAngularDamping=true
    bUseGravityOverride=true
    gravityScale=0.35
    simulationSpace=BoneRelative
    alphaBoolBlend {blendOption=Linear}
    boundBone {boneName=dyn_short_fr_l_01}
    componentAppliedLinearAccClamp=<2000 2000 2000>
    componentLinearAccScale=<0.6 0.6 0.6>
    componentLinearVelScale=<0.6 0.6 0.6>
    gravityOverride {z=-980}
    physicsBodyDefinitions [{
      boundBone {boneName=dyn_short_fr_l_01}
      boxExtents=<20 30 30>
      constraintSetup {angularLimitsMax=<0 0 15>, angularLimitsMin=<0 -25 -15>, angularTarget=<1 0 0>}
      localJointOffset=<6 0 0>
     }]
    relativeSpaceBone {boneName=pelvis}
   }
   n27 = Anim Dynamics - Bone: dyn_purse_01_l(ComponentPose=@n26, Alpha=0.5, ExternalForce="0,0,0") {
    angularDampingOverride=0.9999
    angularSpringConstant=20000
    bAngularSpring=true
    bOverrideAngularDamping=true
    bUseGravityOverride=true
    gravityScale=0.35
    alphaBoolBlend {blendOption=Linear}
    boundBone {boneName=dyn_purse_01_l}
    componentAppliedLinearAccClamp=<2000 2000 2000>
    componentLinearAccScale=<0.5 0.5 0.5>
    componentLinearVelScale=<1 1 1>
    gravityOverride {z=-980}
    physicsBodyDefinitions [{
      boundBone {boneName=dyn_purse_01_l}
      boxExtents=<20 30 30>
      constraintSetup {angularLimitsMax=<0 0 0>, angularLimitsMin=<-3 -30 -3>, angularTarget=<1 0 0>}
      localJointOffset=<12 0 0>
     }]
    relativeSpaceBone {boneName=pelvis}
   }
   n28 = Anim Dynamics - Bone: dyn_pendant_bk_01_r(ComponentPose=@n27, Alpha=1, ExternalForce="0,0,0") {
    angularDampingOverride=0.9999
    angularSpringConstant=20000
    bAngularSpring=true
    bOverrideAngularDamping=true
    bUseGravityOverride=true
    gravityScale=0.35
    simulationSpace=BoneRelative
    alphaBoolBlend {blendOption=Linear}
    boundBone {boneName=dyn_pendant_bk_01_r}
    componentAppliedLinearAccClamp=<3000 3000 3000>
    componentLinearAccScale=<2 2 2>
    componentLinearVelScale=<2 2 2>
    gravityOverride {z=-980}
    physicsBodyDefinitions [{
      boundBone {boneName=dyn_pendant_bk_01_r}
      boxExtents=<30 30 30>
      constraintSetup {angularLimitsMax=<2 10 30>, angularLimitsMin=<-2 -30 -30>, angularTarget=<1 0 0>}
      localJointOffset=<4 0 0>
     }]
    relativeSpaceBone {boneName=root}
   }
   n29 = Anim Dynamics - Bone: dyn_pendant_bk_02_r(ComponentPose=@n28, Alpha=1, ExternalForce="0,0,0") {
    angularDampingOverride=0.9999
    angularSpringConstant=20000
    bAngularSpring=true
    bOverrideAngularDamping=true
    bUseGravityOverride=true
    gravityScale=0.35
    simulationSpace=BoneRelative
    alphaBoolBlend {blendOption=Linear}
    boundBone {boneName=dyn_pendant_bk_02_r}
    componentAppliedLinearAccClamp=<3000 3000 3000>
    componentLinearAccScale=<2 2 2>
    componentLinearVelScale=<2 2 2>
    gravityOverride {z=-980}
    physicsBodyDefinitions [{
      boundBone {boneName=dyn_pendant_bk_02_r}
      boxExtents=<30 30 30>
      constraintSetup {angularLimitsMax=<2 15 30>, angularLimitsMin=<-2 -30 -30>, angularTarget=<1 0 0>}
      localJointOffset=<4 0 0>
     }]
    relativeSpaceBone {boneName=root}
   }
   Save cached pose 'AnimDynamicsCache'(Pose=Component To Local(ComponentPose=@n29)) {cacheName=AnimDynamicsCache}
   n30 = Control Rig!(
    Source=Copy Pose From Mesh() {bUseAttachedParent=true}
    Alpha
    Crouching:bool=Get bIsCrouching()
    Melee:bool=Get bIsPlayingMeleeAnim()
    SkyDiveMode:bool=Get bIsSkydiveDiveMode()
    Skydiving:bool=Get bIsSkyDiving()
    ArmBackInMelee:double=Get Alpha_DetectedArmBackR_InMelee() -> (Alpha_DetectedArmBackR_InMelee:double)
    isParachute:bool=Get bIsParachuteOpened()
    TacticalSprintingJump:double=Get Alpha_TacticalSprintingJump() -> (Alpha_TacticalSprintingJump:double)
    RBEnableAndGrindRail:bool=Get RB DIsable / OnGrindRail() -> (RB DIsable / OnGrindRail:bool)
    RailSeated:double=Get RBAN Enable_Rail() -> (RBAN Enable_Rail:double)
   ) -> (Pose) {
    alphaBoolBlend {blendOption=Linear}
    controlRigAssetReference {blueprintRigClass=&/F_MED_LiftingRays_CtrlRig.F_MED_LiftingRays_CtrlRig_C}
   }
   Save cached pose 'CtrlRigCache'(Pose=@n30) {cacheName=CtrlRigCache}
   Get bRigidBodyEnabled()
   Get bIsOnGrindRail()
  ]}
 EventGraph {nodes [
   Event BlueprintUpdateAnimation()
   > n1 = Sequence!(execute) -> (then_0, then_1, then_2:Exec, then_3:Exec, then_4:Exec, then_5:Exec)
   > n2 = "Gravity Override State Function (Parameter Asset) (Nativized)"(
    CharacterPartAnimInstance=Self-Reference()
    GameTimeInSeconds=GetGameTimeInSeconds()
    DebugJointName=spine_05
    Parameters=&/M_MED_LiftingRays_Parameters
   )
   > Set Gravity Override(Gravity Override=@n2)
   > Set RBAN Settings Out(RBAN Settings Out=@n2.OutRigidBodyAnimNodeParameters)
   n3 = Branch(execute=@n1.then_1, Condition=Get bIsCrouching())
   > Set Alpha_DynStanding /Crouch(Alpha_DynStanding /Crouch:double=0.5) -> (Output_Get:double)
   Set Alpha_DynStanding /Crouch(execute=@n3.else, Alpha_DynStanding /Crouch=1)
   n4 = NOT Boolean(A=Get bIsSkyDiving())
   n5 = NOT Boolean(A=Get bIsSurfaceSwimming())
   n6 = OR Boolean(A=Get bIsJumping(), B=Get bIsFalling())
   Set Alpha_JumpingAndFall(execute=@n1.then_2, Alpha_JumpingAndFall:double="To Float (Boolean)"(InBool=AND Boolean!(A=@n6, B=@n4, C:bool=@n5) -> (ReturnValue))) -> (Output_Get:double)
   n7 = NOT Boolean(A=Get bIsSkyDiving())
   n8 = Get bIsFalling()
   > Set Alpha_Falling(Alpha_Falling:double="To Float (Boolean)"(InBool=AND Boolean!(A=@n8, B=@n7) -> (ReturnValue))) -> (Output_Get:double)
   n9 = Get bIsJumping()
   n10 = Get bIsTacticalSprinting()
   > Set Alpha_TacticalSprintingJump(Alpha_TacticalSprintingJump:double="To Float (Boolean)"(InBool=AND Boolean(A=@n9, B=@n10))) -> (Output_Get:double)
   n11 = Get bIsPlayingEmote()
   n12 = Get Selected_Pickaxe() -> (Selected_Pickaxe:bool)
   > Set IdleFE(IdleFE:bool=AND Boolean!(A=Get bIsInFrontEnd(), B=NOT Boolean(A=@n11), C:bool=NOT Boolean(A=@n12)) -> (ReturnValue)) -> (Output_Get:bool)
   n13 = Get bIsOnGrindRail()
   > Set RB DIsable / OnGrindRail(RB DIsable / OnGrindRail:bool=AND Boolean!(A=Get bRigidBodyEnabled(), B=NOT Boolean(A=@n13)) -> (ReturnValue)) -> (Output_Get:bool)
   n14 = GetOwnerSkeletalMeshComponent()
   n15 = BreakRotator(InRot=BreakTransform(InTransform=GetDeltaTransformFromRefPose(self=@n14, BoneName=spine_01, BaseName=pelvis)).Rotation)
   n16 = Branch(execute=@n1.then_4, Condition=AND Boolean(A=Get bIsOnGrindRail(), B="float > float"(A=@n15.Yaw, B=8)))
   > Set RBAN Enable_Rail(RBAN Enable_Rail:double=1) -> (Output_Get:double)
   > Set AlphaRail(AlphaRail:double=0) -> (Output_Get:double)
   Set RBAN Enable_Rail(execute=@n16.else, RBAN Enable_Rail=0)
   > Set AlphaRail(AlphaRail=1)
   n17 = Cast To FortPlayerAnimInstance(execute=[@n22.then, @n1.then_5], Object=GetAnimInstance(self=GetOwnerSkeletalMeshComponent()))
   > Set Selected_Pickaxe(Selected_Pickaxe:bool=false) -> (Output_Get:bool)
   > n18 = Switch on EFortWeaponCoreAnimation!(execute, Selection:EFortWeaponCoreAnimation Enum=Get WeaponCoreAnim(self=@n17)) -> (
    Melee:Exec
    Pistol:Exec
    Shotgun:Exec
    PaperBlueprint:Exec
    Rifle:Exec
    MeleeOneHand:Exec
    MachinePistol:Exec
    RocketLauncher:Exec
    GrenadeLauncher:Exec
    GoingCommando:Exec
    AssaultRifle:Exec
    TacticalShotgun:Exec
    SniperRifle:Exec
    TrapPlacement:Exec
    ShoulderLauncher:Exec
    AbilityDecoTool:Exec
    Crossbow:Exec
    C4:Exec
    RemoteControl:Exec
    DualWield:Exec
    AR_BullPup:Exec
    AR_ForwardGrip:Exec
    MedPackPaddles:Exec
    SMG_P90:Exec
    AR_DrumGun:Exec
    Consumable_Small:Exec
    Consumable_Large:Exec
    Balloon:Exec
    MountedTurret:Exec
    CreativeTool:Exec
    ExplosiveBow:Exec
    AshtonIndigo:Exec
    AshtonChicago:Exec
    MeleeDualWield:Exec
    Unarmed:Exec
   )
   Set Selected_Pickaxe(execute=[@n18.Melee, @n18.MeleeDualWield], Selected_Pickaxe=true)
   n19 = GetOwnerSkeletalMeshComponent()
   n20 = BreakRotator(InRot=BreakTransform(InTransform=GetDeltaTransformFromRefPose(self=@n19, BoneName=spine_05, BaseName=pelvis)).Rotation)
   n21 = Branch(
    execute=@n1.then_3
    Condition=AND Boolean!(A="float < float"(A=@n20, B=-30), B=Get bIsPlayingMeleeAnim(), C:bool=Get Selected_Pickaxe()) -> (ReturnValue)
   )
   > Set Alpha_DetectedArmBackR_InMelee(Alpha_DetectedArmBackR_InMelee:double=1) -> (Output_Get:double)
   Set Alpha_DetectedArmBackR_InMelee(execute=@n21.else, Alpha_DetectedArmBackR_InMelee=0)
   n22 = Event OnWeaponChanged()
   "float > float"(A=@n20, B=-20)
  ]}
]
variables [
 Alpha_DetectedArmBackR_InMelee {type=number}
 Alpha_DynStanding /Crouch {type=number}
 Alpha_Falling {type=number}
 Alpha_JumpingAndFall {type=number}
 Alpha_TacticalSprintingJump {type=number}
 AlphaRail {type=number}
 Gravity Override {type=Vector}
 Gravity Override Params {type=GravityOverrideParamsStruct}
 IdleFE {type=boolean}
 RB DIsable / OnGrindRail {type=boolean}
 RBAN Enable_Rail {type=number}
 RBAN Params {
  type=RBANParamsStruct
  default {
   dBNO {simSpaceSettings {
     maxAngularAcceleration=1000
     maxAngularVelocity=3
     maxLinearAcceleration=300
     maxLinearVelocity=700
     velocityScaleZ=0
     worldAlpha=0.85
     externalLinearDragV=<0 0.5 0>
    }}
   dBNO_Carried {simSpaceSettings {
     maxAngularAcceleration=1000
     maxAngularVelocity=3
     maxLinearAcceleration=300
     maxLinearVelocity=700
     velocityScaleZ=0
     externalLinearDragV=<0 0.5 0>
    }}
   emoteOrMelee {simSpaceSettings {
     maxAngularAcceleration=1000
     maxAngularVelocity=3
     maxLinearAcceleration=300
     maxLinearVelocity=700
     velocityScaleZ=0
     externalLinearDragV=<0 0.5 0>
    }}
   front_End {simSpaceSettings {
     maxAngularAcceleration=1000
     maxAngularVelocity=3
     maxLinearAcceleration=300
     maxLinearVelocity=700
     velocityScaleZ=0
     worldAlpha=0.3
     externalLinearDragV=<0 0 0>
    }}
   onGround_Crouch_Moving {simSpaceSettings {
     maxAngularAcceleration=1000
     maxAngularVelocity=3
     maxLinearAcceleration=300
     maxLinearVelocity=700
     velocityScaleZ=0
     externalLinearDragV=<0 0 0>
    }}
   onGround_Falling {simSpaceSettings {
     maxAngularAcceleration=1000
     maxAngularVelocity=3
     maxLinearAcceleration=300
     maxLinearVelocity=700
     velocityScaleZ=0
     externalLinearDragV=<0 0.5 0>
    }}
   onGround_Moving {
    componentLinearAccScale=<0 0 0>
    componentLinearVelScale=<0 0 0>
    simSpaceSettings {
     dampingAlpha=2
     maxAngularAcceleration=700
     maxAngularVelocity=700
     maxLinearAcceleration=1000
     maxLinearVelocity=700
     velocityScaleZ=0.5
     worldAlpha=0.35
     externalLinearDragV=<2 2 2>
    }
   }
   onGround_Standing {simSpaceSettings {
     maxAngularAcceleration=1000
     maxAngularVelocity=3
     maxLinearAcceleration=300
     maxLinearVelocity=700
     velocityScaleZ=0
     externalLinearDragV=<0 0.5 0>
    }}
   rail_Grinding {simSpaceSettings {
     maxAngularAcceleration=1000
     maxAngularVelocity=3
     maxLinearAcceleration=300
     maxLinearVelocity=700
     velocityScaleZ=0
     worldAlpha=0.1
     externalLinearDragV=<0 0.5 0>
    }}
   riding_Moving {simSpaceSettings {
     maxAngularAcceleration=1000
     maxAngularVelocity=3
     maxLinearAcceleration=300
     maxLinearVelocity=700
     velocityScaleZ=0
     externalLinearDragV=<0 0.5 0>
    }}
   riding_Moving_Targeting {simSpaceSettings {
     maxAngularAcceleration=1000
     maxAngularVelocity=3
     maxLinearAcceleration=300
     maxLinearVelocity=700
     velocityScaleZ=0
     externalLinearDragV=<0 0.5 0>
    }}
   riding_Standing {simSpaceSettings {
     maxAngularAcceleration=1000
     maxAngularVelocity=3
     maxLinearAcceleration=300
     maxLinearVelocity=700
     velocityScaleZ=0
     externalLinearDragV=<0 0.5 0>
    }}
   skydive_DivingDown {simSpaceSettings {
     maxAngularAcceleration=1000
     maxAngularVelocity=3
     maxLinearAcceleration=300
     maxLinearVelocity=700
     velocityScaleZ=0
     externalLinearDragV=<0 0.5 0>
    }}
   skydive_FloatingDown {
    componentLinearAccScale=<0 0 0>
    componentLinearVelScale=<0 0 0>
    simSpaceSettings {
     dampingAlpha=0.3
     maxAngularAcceleration=1000
     maxAngularVelocity=1.5
     maxLinearAcceleration=300
     maxLinearVelocity=700
     velocityScaleZ=0.2
     worldAlpha=0.2
     externalLinearDragV=<0 0 0>
    }
   }
   skydive_GoingUp {simSpaceSettings {
     maxAngularAcceleration=1000
     maxAngularVelocity=3
     maxLinearAcceleration=300
     maxLinearVelocity=700
     velocityScaleZ=0
     worldAlpha=0.5
     externalLinearDragV=<0 0.5 0>
    }}
   skydive_Parachuting {
    componentLinearAccScale=<0 0.1 0.3>
    componentLinearVelScale=<0 0.1 0.3>
    simSpaceSettings {
     maxAngularAcceleration=1000
     maxAngularVelocity=3
     maxLinearAcceleration=300
     maxLinearVelocity=700
     velocityScaleZ=0
     externalLinearDragV=<0 0.5 0>
    }
   }
   skydive_Surfing {simSpaceSettings {
     maxAngularAcceleration=1000
     maxAngularVelocity=3
     maxLinearAcceleration=300
     maxLinearVelocity=700
     velocityScaleZ=0
     externalLinearDragV=<0 0.5 0>
    }}
   swimming {simSpaceSettings {
     maxAngularAcceleration=1000
     maxAngularVelocity=3
     maxLinearAcceleration=300
     maxLinearVelocity=700
     velocityScaleZ=0
     externalLinearDragV=<0 0.5 0>
    }}
   tactical_Sprint {simSpaceSettings {
     maxAngularAcceleration=1000
     maxAngularVelocity=3
     maxLinearAcceleration=300
     maxLinearVelocity=700
     velocityScaleZ=0
     externalLinearDragV=<0 0.5 0>
    }}
   ziplining {simSpaceSettings {
     maxAngularAcceleration=1000
     maxAngularVelocity=3
     maxLinearAcceleration=300
     maxLinearVelocity=700
     velocityScaleZ=0
     externalLinearDragV=<0 0.5 0>
    }}
  }
 }
 RBAN Settings Out {type=FortRigidBodyAnimNodeParameters}
 Selected_Pickaxe {type=boolean}
 SimSpaceSettingsOLD {
  type=SimSpaceSettings
  default {
   dampingAlpha=1
   maxAngularAcceleration=10000
   maxAngularVelocity=10000
   maxLinearAcceleration=10000
   maxLinearVelocity=10000
   velocityScaleZ=1
   worldAlpha=0
   externalLinearDragV=<0 0 0>
  }
 }
]
