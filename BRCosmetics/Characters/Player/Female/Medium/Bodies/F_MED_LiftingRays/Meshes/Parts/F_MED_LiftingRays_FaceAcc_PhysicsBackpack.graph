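# Physics asset for the LiftingRays female medium FaceAcc part, named as its backpack variant, giving capsule bodies to dynamic hair bones and the upper body.
# Front, side and extra strands are constrained to the head, base bones or earlier strand bones, with locked twist and mostly tight swing limits.
# Capsules on head, neck, spine, clavicles and upper arms provide collision shapes.
# The hairB and left chest-fur bodies have no constraints.
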PhysicsAsset /BRCosmetics/Characters/Player/Female/Medium/Bodies/F_MED_LiftingRays/Meshes/Parts/F_MED_LiftingRays_FaceAcc_PhysicsBackpack
bodies {
 base_hairB_C_01 [{shapeType=Capsule, length=0.505, radius=0.505, shapeName=base_hairB_C_01_capsule, rotation=<rot 0 -90 0>}]
 base_hairB_L_01 [{shapeType=Capsule, length=0.505, radius=0.505, shapeName=base_hairB_L_01_capsule, rotation=<rot 0 -90 0>}]
 base_hairB_R_01 [{shapeType=Capsule, length=0.505, radius=0.505, shapeName=base_hairB_R_01_capsule, rotation=<rot 0 -90 0>}]
 base_hair_fr_a_l [{shapeType=Capsule, length=0.505, radius=0.505, shapeName=base_hair_fr_a_l_capsule, rotation=<rot 0 -90 0>}]
 base_hair_fr_a_r [{shapeType=Capsule, length=0.505, radius=0.505, shapeName=base_hair_fr_a_r_capsule, rotation=<rot 0 -90 0>}]
 base_sd_extra_l [{shapeType=Capsule, length=0.505, radius=0.505, shapeName=base_sd_extra_l_capsule, rotation=<rot 0 -90 0>}]
 clavicle_l [{shapeType=Capsule, length=10.6, radius=7, shapeName=clavicle_l_capsule, center=<6.2526 0.1462 -2.2573>, rotation=<rot 0 -90 90>}]
 clavicle_r [
  {shapeType=Capsule, length=10.6, radius=5, shapeName=clavicle_l_capsule, center=<-6.2526 -0.1462 2.2573>, rotation=<rot 0 -90 90>}
  {shapeType=Capsule, length=3.8, radius=4.2, shapeName=clavicle_r_capsule, center=<-6.1037 -0.1074 -2.9779>, rotation=<rot 90 0 0>}
 ]
 dyn_ChestFur_03_l [{
   shapeType=Capsule
   length=2.1393
   radius=4.1911
   shapeName=dyn_ChestFur_03_r_capsule
   center=<4.7046 0.2401 -1.7511>
   rotation=<rot 81.5197 -179.9987 -179.9987>
  }]
 dyn_hairB_C_01 [{shapeType=Capsule, length=6.0314, radius=1, shapeName=dyn_hairB_C_01_capsule, center=<3.9112 0 0>, rotation=<rot 90 0 0>}]
 dyn_hairB_L_01 [{shapeType=Capsule, length=6.1739, radius=1, shapeName=dyn_hairB_L_01_capsule, center=<3.7963 0 0>, rotation=<rot 90 0 0>}]
 dyn_hairB_R_01 [{shapeType=Capsule, length=6.5288, radius=1, shapeName=dyn_hairB_R_01_capsule, center=<3.6783 0 0>, rotation=<rot 90 0 0>}]
 dyn_hair_fr_a_01_l [{
   shapeType=Capsule
   length=2.2371
   radius=1.5636
   shapeName=dyn_hair_fr_a_01_l_capsule
   center=<1.0156 -0.6984 -1.826>
   rotation=<rot 0 105.0627 -78.3865>
  }]
 dyn_hair_fr_a_01_r [{
   shapeType=Capsule
   length=1.8401
   radius=1.3197
   shapeName=dyn_hair_fr_a_01_r_capsule
   center=<2.0366 0.812 -0.3293>
   rotation=<rot 7.3669 -90 5.9791>
  }]
 dyn_hair_fr_a_02_l [{
   shapeType=Capsule
   length=4.872
   radius=1.4713
   shapeName=dyn_hair_fr_a_02_l_capsule
   center=<2.8779 -1.6666 1.2466>
   rotation=<rot -28.8671 -90 36.673>
  }]
 dyn_hair_fr_a_02_r [{
   shapeType=Capsule
   length=2.4092
   radius=1
   shapeName=dyn_hair_fr_a_02_r_capsule
   center=<1.4484 0.325 0.6602>
   rotation=<rot -2.092 74.6342 -66.9872>
  }]
 dyn_hair_sd_a_01_l [{
   shapeType=Capsule
   length=2.0417
   radius=1.8257
   shapeName=dyn_hair_sd_a_01_l_capsule
   center=<2.2279 0.2466 -0.2808>
   rotation=<rot 0 -95.2782 -79.3738>
  }]
 dyn_hair_sd_a_01_r [{
   shapeType=Capsule
   length=2.4066
   radius=1.0623
   shapeName=dyn_hair_sd_a_01_r_capsule
   center=<1.4334 0.1818 -0.2731>
   rotation=<rot -1.0794 -90.0806 -0.8787>
  }]
 dyn_hair_sd_a_02_l [{
   shapeType=Capsule
   length=1.691
   radius=0.8586
   shapeName=dyn_hair_sd_a_02_l_capsule
   center=<1.4602 0.642 0.5024>
   rotation=<rot 0 110.6071 -64.1375>
  }]
 dyn_hair_sd_a_02_r [{shapeType=Capsule, length=2, radius=1, shapeName=dyn_hair_sd_a_02_r_capsule, center=<1.7212 0.1731 0>, rotation=<rot 90 0 0>}]
 dyn_hair_sd_a_03_l [{
   shapeType=Capsule
   length=0.9726
   radius=0.6626
   shapeName=dyn_hair_sd_a_03_l_capsule
   center=<0.7271 0.2073 0.1932>
   rotation=<rot 0 105.8251 -86.543>
  }]
 dyn_hair_sd_a_03_r [{
   shapeType=Capsule
   length=2.5
   radius=1
   shapeName=dyn_hair_sd_a_03_r_capsule
   center=<2.9656 -0.1193 0.0866>
   rotation=<rot 77.7578 -179.9998 -166.8766>
  }]
 dyn_hair_sd_a_04_l [{
   shapeType=Capsule
   length=0.8842
   radius=0.4935
   shapeName=dyn_hair_sd_a_04_l_capsule
   center=<0.7987 0.3002 0.0821>
   rotation=<rot 0 -65.8748 -76.9865>
  }]
 dyn_hair_sd_b_01_l [{
   shapeType=Capsule
   length=3
   radius=1
   shapeName=dyn_hair_sd_b_01_l_capsule
   center=<2.4132 -0.3054 -0.1374>
   rotation=<rot 84.0266 90.0001 -179.9999>
  }]
 dyn_hair_sd_b_01_r [{shapeType=Capsule, length=3, radius=1, shapeName=dyn_hair_sd_b_01_r_capsule, center=<1.2768 0 0>, rotation=<rot 90 0 0>}]
 dyn_hair_sd_b_02_l [
  {shapeType=Capsule, length=2.5, radius=1, shapeName=dyn_hair_sd_b_02_l_capsule, center=<0.6153 -0.2085 -0.257>, rotation=<rot 90 -90 0>}
  {
   shapeType=Capsule
   length=1.5
   radius=1
   shapeName=dyn_hair_sd_b_02_l_capsule
   center=<2.3482 0.2401 -1.3176>
   rotation=<rot 61.629 0 -12.7623>
  }
 ]
 dyn_hair_sd_b_02_r [{
   shapeType=Capsule
   length=2.5
   radius=1
   shapeName=dyn_hair_sd_b_02_r_capsule
   center=<2.1007 -0.0003 0.7119>
   rotation=<rot 76.2203 -179.982 -179.9661>
  }]
 dyn_hair_sd_c_01_l [{
   shapeType=Capsule
   length=1.8885
   radius=1.3243
   shapeName=dyn_hair_sd_c_01_l_capsule
   center=<0.8547 0.001 -0.3997>
   rotation=<rot 0 88.9441 -62.7344>
  }]
 dyn_hair_sd_c_02_l [{
   shapeType=Capsule
   length=1.658
   radius=0.6317
   shapeName=dyn_hair_sd_c_02_l_capsule
   center=<1.3282 0.0768 0.7135>
   rotation=<rot 0 89.5132 -50.6096>
  }]
 dyn_sd_extra_l [{shapeType=Capsule, length=0.505, radius=0.505, shapeName=dyn_sd_extra_l_capsule, rotation=<rot 0 -90 0>}]
 head [{shapeType=Capsule, length=2.6403, radius=7.3597, center=<2.5678 1.1032 0>, rotation=<rot 90 0 0>}]
 neck_01 [{shapeType=Capsule, length=12.5, radius=6, shapeName=neck_01_capsule, center=<1.6583 0.6206 0>}]
 spine_02 [{shapeType=Capsule, length=16.7, radius=9.3, shapeName=spine_02_capsule, center=<6.0567 1.2153 0>}]
 spine_03 [{shapeType=Capsule, length=16.7, radius=9.3, shapeName=spine_02_capsule, center=<6.0567 1.1414 0>}]
 spine_04 [
  {shapeType=Capsule, length=17.8, radius=9.1, shapeName=spine_04_capsule, center=<4.0135 1.6435 0>, rotation=<rot 0 -90 0>}
  {shapeType=Capsule, length=18.5, radius=9, shapeName=spine_04_capsule, center=<8.6672 1.2123 0>}
 ]
 spine_05 [{shapeType=Capsule, length=17.2, radius=8, shapeName=spine_05_capsule, center=<2.8857 0.6609 0>, rotation=<rot 0 -90 0>}]
 upperarm_l [{shapeType=Capsule, length=13.5146, radius=5.3875, shapeName=upperarm_l_capsule, center=<11.2314 -0.3627 0>, rotation=<rot 90 0 0>}]
 upperarm_r [{shapeType=Capsule, length=13.5146, radius=5.3875, shapeName=upperarm_l_capsule, center=<-11.2314 0.3627 0>, rotation=<rot 90 0 0>}]
}
constraints [
 {bone1Name=dyn_hair_sd_b_01_l, bone2Name=head, swing1LimitDegrees=50, swing2LimitDegrees=5, twistLimitDegrees=45, twistMotion=Locked}
 {
  bone1Name=dyn_hair_sd_b_02_l
  bone2Name=dyn_hair_sd_b_01_l
  swing1LimitDegrees=50
  swing2LimitDegrees=10
  twistLimitDegrees=45
  twistMotion=Locked
 }
 {bone1Name=dyn_hair_sd_b_01_r, bone2Name=head, swing1LimitDegrees=10, swing2LimitDegrees=35, twistLimitDegrees=45, twistMotion=Locked}
 {
  bone1Name=dyn_hair_sd_b_02_r
  bone2Name=dyn_hair_sd_b_01_r
  swing1LimitDegrees=10
  swing2LimitDegrees=15
  twistLimitDegrees=45
  twistMotion=Locked
 }
 {bone1Name=dyn_hair_sd_a_01_r, bone2Name=head, swing1LimitDegrees=10, swing2LimitDegrees=20, twistLimitDegrees=45, twistMotion=Locked}
 {
  bone1Name=dyn_hair_sd_a_02_r
  bone2Name=dyn_hair_sd_a_01_r
  swing1LimitDegrees=10
  swing2LimitDegrees=20
  twistLimitDegrees=45
  twistMotion=Locked
 }
 {
  bone1Name=dyn_hair_sd_a_03_r
  bone2Name=dyn_hair_sd_a_01_r
  swing1LimitDegrees=10
  swing2LimitDegrees=20
  twistLimitDegrees=45
  twistMotion=Locked
 }
 {
  bone1Name=dyn_hair_fr_a_01_l
  bone2Name=base_hair_fr_a_l
  swing1LimitDegrees=5
  swing2LimitDegrees=5
  twistLimitDegrees=45
  twistMotion=Locked
 }
 {
  bone1Name=dyn_hair_fr_a_02_l
  bone2Name=dyn_hair_fr_a_01_l
  swing1LimitDegrees=3
  swing2LimitDegrees=5
  twistLimitDegrees=45
  twistMotion=Locked
 }
 {
  bone1Name=dyn_hair_fr_a_01_r
  bone2Name=base_hair_fr_a_r
  swing1LimitDegrees=10
  swing2LimitDegrees=20
  twistLimitDegrees=45
  twistMotion=Locked
 }
 {
  bone1Name=dyn_hair_fr_a_02_r
  bone2Name=dyn_hair_fr_a_01_r
  swing1LimitDegrees=10
  swing2LimitDegrees=20
  twistLimitDegrees=45
  twistMotion=Locked
 }
 {bone1Name=dyn_hair_sd_c_01_l, bone2Name=head, swing1LimitDegrees=10, swing2LimitDegrees=10, twistLimitDegrees=45, twistMotion=Locked}
 {
  bone1Name=dyn_hair_sd_c_02_l
  bone2Name=dyn_hair_sd_c_01_l
  swing1LimitDegrees=10
  swing2LimitDegrees=10
  twistLimitDegrees=45
  twistMotion=Locked
 }
 {bone1Name=dyn_hair_sd_a_01_l, bone2Name=head, swing1LimitDegrees=5, swing2LimitDegrees=10, twistLimitDegrees=45, twistMotion=Locked}
 {
  bone1Name=dyn_hair_sd_a_02_l
  bone2Name=dyn_hair_sd_a_01_l
  swing1LimitDegrees=2
  swing2LimitDegrees=10
  twistLimitDegrees=45
  twistMotion=Locked
 }
 {
  bone1Name=dyn_hair_sd_a_03_l
  bone2Name=dyn_hair_sd_a_01_l
  swing1LimitDegrees=2
  swing2LimitDegrees=10
  twistLimitDegrees=45
  twistMotion=Locked
 }
 {
  bone1Name=dyn_hair_sd_a_04_l
  bone2Name=dyn_hair_sd_a_01_l
  swing1LimitDegrees=2
  swing2LimitDegrees=10
  twistLimitDegrees=45
  twistMotion=Locked
 }
 {bone1Name=dyn_sd_extra_l, bone2Name=base_sd_extra_l, swing1LimitDegrees=24, swing2LimitDegrees=0, twistLimitDegrees=45, twistMotion=Locked}
]
