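# Post-process style AnimBP for the LiftingRays female face-accessory part, animating its hair with state-aware secondary motion.
# Copies parent and body mesh poses, then runs a Control Rig fed movement, weapon, glider and emote states.
# That cached pose feeds crouch, normal and backbling/cape/no-RBAN branches, each with its own RigidBody physics asset, plus corrective bone offsets and crouch-only AnimDynamics.
# The event graph computes those states and a parameter-driven gravity override.
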AnimBlueprint /BRCosmetics/Characters/Player/Female/Medium/Bodies/F_MED_LiftingRays/Meshes/Parts/F_MED_LiftingRays_FaceAcc_AnimBP
target_skeleton=&/F_MED_LiftingRays_FaceAcc_Skeleton
graphs [
 AnimGraph {nodes [
   n1 = Blend Poses by bool(BlendPose_0=Use cached pose 'SetupCrouchingCache'(), BlendPose_1=Use cached pose 'SetupNormalCache'(), bActiveValue=Get bIsCrouching()) {blendType=Linear}
   n2 = Use cached pose 'WearingBackpackCache'()
   n3 = Get isWearingBackbling / Cape / NoRBAN() -> (isWearingBackbling / Cape / NoRBAN:bool)
   n4 = "Transform (Modify) Bone - Bone: dyn_hairB_L_02"(
    ComponentPose=Local To Component(LocalPose=Blend Poses by bool(BlendPose_0=@n2, BlendPose_1=@n1, bActiveValue=@n3) {blendType=Linear})
    Alpha=1
    Translation="0.247,-1.404,-0.000"
    Rotation="0.003,12,0.011"
    Scale="1,1,1"
   ) {
    rotationMode=BMM_Additive
    rotationSpace=BCS_BoneSpace
    scaleMode=BMM_Additive
    scaleSpace=BCS_BoneSpace
    translationMode=BMM_Additive
    translationSpace=BCS_BoneSpace
    alphaBoolBlend {blendOption=Linear}
    alphaScaleBiasClamp {bInterpResult=true, interpSpeedIncreasing=1}
    boneToModify {boneName=dyn_hairB_L_02}
    rotation=<rot 0.003 12 0.011>
    translation=<0.247 -1.404>
   }
   n5 = "Transform (Modify) Bone - Bone: dyn_hairB_C_02"(ComponentPose=@n4, Alpha=1, Translation="0.288,-1.592,0.252", Rotation="-0.005,10.745,0.006", Scale="1,1,1") {
    rotationMode=BMM_Additive
    rotationSpace=BCS_BoneSpace
    scaleMode=BMM_Additive
    scaleSpace=BCS_BoneSpace
    translationMode=BMM_Additive
    translationSpace=BCS_BoneSpace
    alphaBoolBlend {blendOption=Linear}
    alphaScaleBiasClamp {bInterpResult=true, interpSpeedIncreasing=1}
    boneToModify {boneName=dyn_hairB_C_02}
    rotation=<rot -0.005 10.745 0.006>
    translation=<0.288 -1.592 0.252>
   }
   n6 = "Transform (Modify) Bone - Bone: dyn_hairB_R_02"(ComponentPose=@n5, Alpha=1, Translation="-0.305,-3.240,0.772", Rotation="0.002,16,-0.01", Scale="1,1,1") {
    rotationMode=BMM_Additive
    rotationSpace=BCS_BoneSpace
    scaleMode=BMM_Additive
    scaleSpace=BCS_BoneSpace
    translationMode=BMM_Additive
    translationSpace=BCS_BoneSpace
    alphaBoolBlend {blendOption=Linear}
    alphaScaleBiasClamp {bInterpResult=true, interpSpeedIncreasing=1}
    boneToModify {boneName=dyn_hairB_R_02}
    rotation=<rot 0.002 16 -0.01>
    translation=<-0.305 -3.24 0.772>
   }
   Output Pose(Result=Component To Local(ComponentPose=@n6)) {layerGroup=None}
   n7 = Layered blend per bone(
    BasePose=Copy Pose From Mesh() {bUseAttachedParent=true}
    BlendPoses_0=Copy Pose From Mesh(SourceMeshComponent=Get Body_Mesh()) {bUseAttachedParent=true}
   ) {layerSetup [{branchFilters [{boneName=clavicle_l}, {boneName=clavicle_r}, {boneName=upperarm_r}, {boneName=upperarm_l}]}]}
   n8 = Control Rig!(
    Source=@n7
    Alpha
    Crouching_Bool:bool=Get bIsCrouching()
    Pistol:double=Get Pistol() -> (Pistol:double)
    Floating:bool=Get bIsSkydiveFloating()
    Moving2d:bool=Get bIsMoving2D()
    WearingBackblingAndNoRBAN:bool=Get isWearingBackbling / Cape / NoRBAN()
    Glider:double=Get HangGlidingAlpha() -> (HangGlidingAlpha:double)
    FallingSwim:double=Get FallingSwimming() -> (FallingSwimming:double)
    CelebrationEmoteNoRban:double=Get PlayingCelebrationNoRban() -> (PlayingCelebrationNoRban:double)
    CelebrationEmoteRban:double=Get PlayingCelebrationRban() -> (PlayingCelebrationRban:double)
    Skydiving:bool=Get bIsSkyDiving()
    PlayEmote:bool=Get bIsPlayingEmote()
   ) -> (Pose) {
    alphaBoolBlend {blendOption=Linear}
    controlRigAssetReference {blueprintRigClass=&/F_MED_LiftingRays_FaceAcc_CtrlRig.F_MED_LiftingRays_FaceAcc_CtrlRig_C}
   }
   Save cached pose 'ControlRig Cache'(Pose=@n8) {cacheName=ControlRig Cache}
   n9 = RigidBody!(
    ComponentPose=Local To Component(LocalPose=Use cached pose 'ControlRig Cache'())
    Alpha
    OverrideWorldGravity:Vector=Get Gravity Override()
    SimSpaceSettings:Sim Space Settings Structure=(WorldAlpha=0.3,VelocityScaleZ=1,DampingAlpha=1,MaxLinearVelocity=700,MaxAngularVelocity=15,MaxLinearAcceleration=300,MaxAngularAcceleration=1000,ExternalLinearDragV=(X=0,Y=0,Z=0),ExternalLinearVelocity=(X=0,Y=0,Z=0),ExternalAngularVelocity=(X=0,Y=0,Z=0))
   ) -> (Pose) {
    bClampLinearTranslationLimitToRefPose=true
    bOverrideWorldGravity=true
    overridePhysicsAsset=&/F_MED_LiftingRays_FaceAcc_PhysicsCrouch
    simulationSpace=BaseBoneSpace
    alphaBoolBlend {blendOption=Linear}
    baseBoneRef {boneName=head}
    componentAppliedLinearAccClamp=<3000 3000 3000>
    simSpaceSettings {maxAngularAcceleration=1000, maxAngularVelocity=15, maxLinearAcceleration=300, maxLinearVelocity=700, worldAlpha=0.3}
   }
   n10 = "Transform (Modify) Bone - Bone: base_hairB_L_01"!(ComponentPose=@n9, bAlphaBoolEnabled:bool=True, Translation="0,0,0", Rotation="0,-10,0", Scale="1,1,1") -> (Pose) {
    alphaInputType=Bool
    rotationMode=BMM_Additive
    rotationSpace=BCS_BoneSpace
    scaleMode=BMM_Additive
    scaleSpace=BCS_BoneSpace
    translationMode=BMM_Additive
    translationSpace=BCS_BoneSpace
    alphaBoolBlend {blendOption=Linear}
    alphaScaleBiasClamp {bInterpResult=true, interpSpeedIncreasing=1}
    boneToModify {boneName=base_hairB_L_01}
    rotation {yaw=-10}
   }
   n11 = Anim Dynamics - Bone: dyn_hair_sd_d_03_r(ComponentPose=@n10, Alpha=1, ExternalForce="0,0,0") {
    angularDampingOverride=1
    angularSpringConstant=20000
    bAngularSpring=true
    bOverrideAngularDamping=true
    bUseGravityOverride=true
    gravityScale=0.35
    simulationSpace=BoneRelative
    alphaBoolBlend {blendOption=Linear}
    boundBone {boneName=dyn_hair_sd_d_03_r}
    componentLinearAccScale=<1 1 1>
    componentLinearVelScale=<1 1 1>
    gravityOverride {z=-980}
    physicsBodyDefinitions [{
      boundBone {boneName=dyn_hair_sd_d_03_r}
      boxExtents=<20 30 30>
      constraintSetup {angularLimitsMax=<0 30 0>, angularLimitsMin=<0 -30 0>, angularTarget=<1 0 0>}
      localJointOffset=<4 0 0>
     }]
    relativeSpaceBone {boneName=head}
   }
   n12 = Anim Dynamics - Bone: dyn_hair_sd_d_02_r(ComponentPose=@n11, Alpha=1, ExternalForce="0,0,0") {
    angularDampingOverride=1
    angularSpringConstant=20000
    bAngularSpring=true
    bOverrideAngularDamping=true
    bUseGravityOverride=true
    gravityScale=0.35
    simulationSpace=BoneRelative
    alphaBoolBlend {blendOption=Linear}
    boundBone {boneName=dyn_hair_sd_d_02_r}
    componentLinearAccScale=<1 1 1>
    componentLinearVelScale=<1 1 1>
    gravityOverride {z=-980}
    physicsBodyDefinitions [{
      boundBone {boneName=dyn_hair_sd_d_02_r}
      boxExtents=<20 30 30>
      constraintSetup {angularLimitsMax=<0 30 0>, angularLimitsMin=<0 -30 0>, angularTarget=<1 0 0>}
      localJointOffset=<4 0 0>
     }]
    relativeSpaceBone {boneName=head}
   }
   n13 = Anim Dynamics - Bone: dyn_hair_sd_d_01_r(ComponentPose=@n12, Alpha=1, ExternalForce="0,0,0") {
    angularDampingOverride=1
    angularSpringConstant=20000
    bAngularSpring=true
    bOverrideAngularDamping=true
    bUseGravityOverride=true
    gravityScale=0.35
    simulationSpace=BoneRelative
    alphaBoolBlend {blendOption=Linear}
    boundBone {boneName=dyn_hair_sd_d_01_r}
    componentLinearAccScale=<1 1 1>
    componentLinearVelScale=<1 1 1>
    gravityOverride {z=-980}
    physicsBodyDefinitions [{
      boundBone {boneName=dyn_hair_sd_d_01_r}
      boxExtents=<20 30 30>
      constraintSetup {angularLimitsMax=<0 30 0>, angularLimitsMin=<0 -30 0>, angularTarget=<1 0 0>}
      localJointOffset=<4 0 0>
     }]
    relativeSpaceBone {boneName=head}
   }
   n14 = Anim Dynamics - Bone: dyn_hair_sd_c_01_r(ComponentPose=@n13, Alpha=1, ExternalForce="0,0,0") {
    angularDampingOverride=1
    angularSpringConstant=20000
    bAngularSpring=true
    bOverrideAngularDamping=true
    bUseGravityOverride=true
    gravityScale=0.35
    simulationSpace=BoneRelative
    alphaBoolBlend {blendOption=Linear}
    boundBone {boneName=dyn_hair_sd_c_01_r}
    componentLinearAccScale=<1 1 1>
    componentLinearVelScale=<1 1 1>
    gravityOverride {z=-980}
    physicsBodyDefinitions [{
      boundBone {boneName=dyn_hair_sd_c_01_r}
      boxExtents=<20 30 30>
      constraintSetup {angularLimitsMax=<0 30 0>, angularLimitsMin=<0 -30 0>, angularTarget=<1 0 0>}
      localJointOffset=<4 0 0>
     }]
    relativeSpaceBone {boneName=head}
   }
   n15 = Anim Dynamics - Bone: dyn_hair_sd_c_02_r(ComponentPose=@n14, Alpha=1, ExternalForce="0,0,0") {
    angularDampingOverride=1
    angularSpringConstant=20000
    bAngularSpring=true
    bOverrideAngularDamping=true
    bUseGravityOverride=true
    gravityScale=0.35
    simulationSpace=BoneRelative
    alphaBoolBlend {blendOption=Linear}
    boundBone {boneName=dyn_hair_sd_c_02_r}
    componentLinearAccScale=<1 1 1>
    componentLinearVelScale=<1 1 1>
    gravityOverride {z=-980}
    physicsBodyDefinitions [{
      boundBone {boneName=dyn_hair_sd_c_02_r}
      boxExtents=<20 30 30>
      constraintSetup {angularLimitsMax=<0 30 0>, angularLimitsMin=<0 -30 0>, angularTarget=<1 0 0>}
      localJointOffset=<4 0 0>
     }]
    relativeSpaceBone {boneName=head}
   }
   n16 = "Transform (Modify) Bone - Bone: base_hairB_C_01"!(ComponentPose=@n15, bAlphaBoolEnabled:bool=True, Translation="0,0,0", Rotation="0,-10,0", Scale="1,1,1") -> (Pose) {
    alphaInputType=Bool
    rotationMode=BMM_Additive
    rotationSpace=BCS_BoneSpace
    scaleMode=BMM_Additive
    scaleSpace=BCS_BoneSpace
    translationMode=BMM_Additive
    translationSpace=BCS_BoneSpace
    alphaBoolBlend {blendInTime=0.1, blendOption=Linear, blendOutTime=0.1}
    alphaScaleBiasClamp {bInterpResult=true}
    boneToModify {boneName=base_hairB_C_01}
    rotation {yaw=-10}
   }
   Save cached pose 'SetupCrouchingCache'(Pose=Component To Local(ComponentPose=@n16)) {cacheName=SetupCrouchingCache}
   n17 = "Transform (Modify) Bone - Bone: base_sd_extra_l"!(
    ComponentPose=Local To Component(LocalPose=Use cached pose 'ControlRig Cache'())
    bAlphaBoolEnabled:bool=True
    Translation="0,0,0"
    Rotation="22.706,-2.219,1.402"
    Scale="1,1,1"
   ) -> (Pose) {
    alphaInputType=Bool
    rotationMode=BMM_Additive
    rotationSpace=BCS_BoneSpace
    translationMode=BMM_Additive
    translationSpace=BCS_BoneSpace
    alphaBoolBlend {blendInTime=0.2, blendOption=Linear, blendOutTime=0.2}
    boneToModify {boneName=base_sd_extra_l}
    rotation=<rot 22.706 -2.219 1.402>
   }
   n18 = "Transform (Modify) Bone - Bone: base_sd_extra_l"!(ComponentPose=@n17, Alpha:float=1, Translation="0,0,0", Rotation="-8,-15,0", Scale="0.85,1,1") -> (Pose) {
    rotationMode=BMM_Additive
    rotationSpace=BCS_BoneSpace
    scaleMode=BMM_Additive
    scaleSpace=BCS_BoneSpace
    translationMode=BMM_Additive
    translationSpace=BCS_BoneSpace
    alphaBoolBlend {blendOption=Linear}
    boneToModify {boneName=base_sd_extra_l}
    rotation {pitch=-8, yaw=-15}
    scale {x=0.85}
   }
   n19 = "Transform (Modify) Bone - Bone: base_hairB_L_01"!(ComponentPose=@n18, Alpha:float=1, Translation="-2.848,-1.982,-0.000", Rotation="-0.000,-15.382,-0.000", Scale="1,1,1") -> (Pose) {
    rotationMode=BMM_Additive
    rotationSpace=BCS_BoneSpace
    scaleMode=BMM_Additive
    scaleSpace=BCS_BoneSpace
    translationMode=BMM_Additive
    translationSpace=BCS_BoneSpace
    alphaBoolBlend {blendOption=Linear}
    alphaScaleBiasClamp {bInterpResult=true, interpSpeedIncreasing=20}
    boneToModify {boneName=base_hairB_L_01}
    rotation=<rot 0 -15.382 0>
    translation=<-2.848 -1.982>
   }
   n20 = "Transform (Modify) Bone - Bone: base_hairB_R_01"(ComponentPose=@n19, Alpha=1, Translation="-3,0,0", Rotation="0,0,0", Scale="1,1,1") {
    rotationMode=BMM_Additive
    rotationSpace=BCS_BoneSpace
    scaleMode=BMM_Additive
    scaleSpace=BCS_BoneSpace
    translationMode=BMM_Additive
    translationSpace=BCS_BoneSpace
    alphaBoolBlend {blendOption=Linear}
    alphaScaleBiasClamp {bInterpResult=true, interpSpeedIncreasing=20}
    boneToModify {boneName=base_hairB_R_01}
    translation {x=-3}
   }
   n21 = "Transform (Modify) Bone - Bone: base_hairB_R_01"!(ComponentPose=@n20, bAlphaBoolEnabled:bool=True, Translation="-2.851,-0.838,-0.000", Rotation="0,0,0", Scale="1,1,1") -> (Pose) {
    alphaInputType=Bool
    rotationMode=BMM_Additive
    rotationSpace=BCS_BoneSpace
    scaleMode=BMM_Additive
    scaleSpace=BCS_BoneSpace
    translationMode=BMM_Additive
    translationSpace=BCS_BoneSpace
    alphaBoolBlend {blendInTime=0.2, blendOption=Linear, blendOutTime=0.3}
    alphaScaleBiasClamp {bInterpResult=true, interpSpeedIncreasing=20}
    boneToModify {boneName=base_hairB_R_01}
    translation=<-2.851 -0.838>
   }
   n22 = RigidBody(
    ComponentPose=@n21
    OverrideWorldGravity=Get Gravity Override()
    SimSpaceSettings=(WorldAlpha=0.3,VelocityScaleZ=1,DampingAlpha=1,MaxLinearVelocity=700,MaxAngularVelocity=15,MaxLinearAcceleration=300,MaxAngularAcceleration=1000,ExternalLinearDragV=(X=0,Y=0,Z=0),ExternalLinearVelocity=(X=0,Y=0,Z=0),ExternalAngularVelocity=(X=0,Y=0,Z=0))
   ) {
    bClampLinearTranslationLimitToRefPose=true
    bOverrideWorldGravity=true
    overridePhysicsAsset=&/F_MED_LiftingRays_FaceAcc_PhysicsBackup
    simulationSpace=BaseBoneSpace
    alphaBoolBlend {blendOption=Linear}
    baseBoneRef {boneName=head}
    componentAppliedLinearAccClamp=<3000 3000 3000>
    simSpaceSettings {maxAngularAcceleration=1000, maxAngularVelocity=15, maxLinearAcceleration=300, maxLinearVelocity=700, worldAlpha=0.3}
   }
   n23 = "Transform (Modify) Bone - Bone: dyn_hairB_R_02"(ComponentPose=@n22, Alpha=1, Translation="-2.492,-3.386,-2.833", Rotation="0,0,0", Scale="1,1,1") {
    rotationMode=BMM_Additive
    rotationSpace=BCS_BoneSpace
    scaleMode=BMM_Additive
    scaleSpace=BCS_BoneSpace
    translationMode=BMM_Additive
    translationSpace=BCS_BoneSpace
    alphaBoolBlend {blendOption=Linear}
    alphaScaleBiasClamp {bInterpResult=true, interpSpeedDecreasing=3}
    boneToModify {boneName=dyn_hairB_R_02}
    translation=<-2.492 -3.386 -2.833>
   }
   n24 = "Transform (Modify) Bone - Bone: base_sd_extra_l"(ComponentPose=@n23, Alpha=1, Translation="0,0,0", Rotation="0.000,-34.760,26.925", Scale="1,1,1") {
    rotationMode=BMM_Additive
    rotationSpace=BCS_BoneSpace
    scaleMode=BMM_Additive
    scaleSpace=BCS_BoneSpace
    translationMode=BMM_Additive
    translationSpace=BCS_BoneSpace
    alphaBoolBlend {blendOption=Linear}
    alphaScaleBiasClamp {bInterpResult=true, interpSpeedDecreasing=3}
    boneToModify {boneName=base_sd_extra_l}
    rotation=<rot 0 -34.76 26.925>
   }
   n25 = "Transform (Modify) Bone - Bone: base_hairB_C_01"!(ComponentPose=@n24, Alpha:float=1, Translation="0,0,0", Rotation="0,-20,0", Scale="1,1,1") -> (Pose) {
    rotationMode=BMM_Additive
    rotationSpace=BCS_BoneSpace
    scaleMode=BMM_Additive
    scaleSpace=BCS_BoneSpace
    translationMode=BMM_Additive
    translationSpace=BCS_BoneSpace
    alphaBoolBlend {blendOption=Linear}
    alphaScaleBiasClamp {bInterpResult=true, interpSpeedIncreasing=2}
    boneToModify {boneName=base_hairB_C_01}
    rotation {yaw=-20}
   }
   n26 = "Transform (Modify) Bone - Bone: base_hairB_L_01"(ComponentPose=@n25, Alpha=1, Translation="0,0,0", Rotation="0,-20,0", Scale="1,1,1") {
    rotationMode=BMM_Additive
    rotationSpace=BCS_BoneSpace
    scaleMode=BMM_Additive
    scaleSpace=BCS_BoneSpace
    translationMode=BMM_Additive
    translationSpace=BCS_BoneSpace
    alphaBoolBlend {blendOption=Linear}
    alphaScaleBiasClamp {bInterpResult=true, interpSpeedIncreasing=2}
    boneToModify {boneName=base_hairB_L_01}
    rotation {yaw=-20}
   }
   n27 = "Transform (Modify) Bone - Bone: base_hairB_R_01"!(ComponentPose=@n26, Alpha:float=1, Translation="0,0,0", Rotation="0,-20,0", Scale="1,1,1") -> (Pose) {
    rotationMode=BMM_Additive
    rotationSpace=BCS_BoneSpace
    scaleMode=BMM_Additive
    scaleSpace=BCS_BoneSpace
    translationMode=BMM_Additive
    translationSpace=BCS_BoneSpace
    alphaBoolBlend {blendOption=Linear}
    alphaScaleBiasClamp {bInterpResult=true, interpSpeedIncreasing=2}
    boneToModify {boneName=base_hairB_R_01}
    rotation {yaw=-20}
   }
   n28 = "Transform (Modify) Bone - Bone: dyn_hairB_C_03"(ComponentPose=@n27, Alpha=1, Translation="0,-1,0.002", Rotation="0,-80,0", Scale="1,1,1") {
    rotationMode=BMM_Additive
    rotationSpace=BCS_BoneSpace
    scaleMode=BMM_Additive
    scaleSpace=BCS_BoneSpace
    translationMode=BMM_Additive
    translationSpace=BCS_BoneSpace
    alphaBoolBlend {blendOption=Linear}
    alphaScaleBiasClamp {bInterpResult=true, interpSpeedIncreasing=1}
    boneToModify {boneName=dyn_hairB_C_03}
    rotation {yaw=-80}
    translation {y=-1, z=0.002}
   }
   n29 = "Transform (Modify) Bone - Bone: dyn_hairB_L_03"(ComponentPose=@n28, Alpha=1, Translation="0,-1,0", Rotation="0,-80,0", Scale="1,1,1") {
    rotationMode=BMM_Additive
    rotationSpace=BCS_BoneSpace
    scaleMode=BMM_Additive
    scaleSpace=BCS_BoneSpace
    translationMode=BMM_Additive
    translationSpace=BCS_BoneSpace
    alphaBoolBlend {blendOption=Linear}
    alphaScaleBiasClamp {bInterpResult=true, interpSpeedIncreasing=1}
    boneToModify {boneName=dyn_hairB_L_03}
    rotation {yaw=-80}
    translation {y=-1}
   }
   n30 = "Transform (Modify) Bone - Bone: dyn_hairB_R_03"(ComponentPose=@n29, Alpha=1, Translation="0,-1,0", Rotation="0,-80,0", Scale="1,1,1") {
    rotationMode=BMM_Additive
    rotationSpace=BCS_BoneSpace
    scaleMode=BMM_Additive
    scaleSpace=BCS_BoneSpace
    translationMode=BMM_Additive
    translationSpace=BCS_BoneSpace
    alphaBoolBlend {blendOption=Linear}
    alphaScaleBiasClamp {bInterpResult=true, interpSpeedIncreasing=1}
    boneToModify {boneName=dyn_hairB_R_03}
    rotation {yaw=-80}
    translation {y=-1}
   }
   n31 = "Transform (Modify) Bone - Bone: dyn_hairB_R_02"(ComponentPose=@n30, Alpha=1, Translation="-0.909,-0.814,-0.004", Rotation="0.875,11.785,-0.556", Scale="1,1,1") {
    rotationMode=BMM_Additive
    rotationSpace=BCS_BoneSpace
    scaleMode=BMM_Additive
    scaleSpace=BCS_BoneSpace
    translationMode=BMM_Additive
    translationSpace=BCS_BoneSpace
    alphaBoolBlend {blendOption=Linear}
    alphaScaleBiasClamp {bInterpResult=true, interpSpeedIncreasing=5}
    boneToModify {boneName=dyn_hairB_R_02}
    rotation=<rot 0.875 11.785 -0.556>
    translation=<-0.909 -0.814 -0.004>
   }
   n32 = "Transform (Modify) Bone - Bone: base_sd_extra_l"(ComponentPose=@n31, Alpha=1, Translation="-1.491,-1.500,0.112", Rotation="0,0,0", Scale="1,1,1") {
    rotationMode=BMM_Additive
    rotationSpace=BCS_BoneSpace
    scaleMode=BMM_Additive
    scaleSpace=BCS_BoneSpace
    translationMode=BMM_Additive
    translationSpace=BCS_BoneSpace
    alphaBoolBlend {blendOption=Linear}
    alphaScaleBiasClamp {bInterpResult=true, interpSpeedIncreasing=5}
    boneToModify {boneName=base_sd_extra_l}
    translation=<-1.491 -1.5 0.112>
   }
   n33 = "Transform (Modify) Bone - Bone: dyn_hairB_C_02"(ComponentPose=@n32, Alpha=1, Translation="0,0,0", Rotation="0.013,9.716,-0.133", Scale="1,1,1") {
    rotationMode=BMM_Additive
    rotationSpace=BCS_BoneSpace
    scaleMode=BMM_Additive
    scaleSpace=BCS_BoneSpace
    translationMode=BMM_Additive
    translationSpace=BCS_BoneSpace
    alphaBoolBlend {blendOption=Linear}
    alphaScaleBiasClamp {bInterpResult=true, interpSpeedIncreasing=20}
    boneToModify {boneName=dyn_hairB_C_02}
    rotation=<rot 0.013 9.716 -0.133>
   }
   n34 = "Transform (Modify) Bone - Bone: dyn_hairB_R_02"!(ComponentPose=@n33, bAlphaBoolEnabled:bool=True, Translation="0,0,0", Rotation="0.000,18.703,0.007", Scale="1,1,1") -> (Pose) {
    alphaInputType=Bool
    rotationMode=BMM_Additive
    rotationSpace=BCS_BoneSpace
    scaleMode=BMM_Additive
    scaleSpace=BCS_BoneSpace
    translationMode=BMM_Additive
    translationSpace=BCS_BoneSpace
    alphaBoolBlend {blendInTime=0.2, blendOption=Linear, blendOutTime=0.3}
    alphaScaleBiasClamp {bInterpResult=true, interpSpeedIncreasing=20}
    boneToModify {boneName=dyn_hairB_R_02}
    rotation=<rot 0 18.703 0.007>
   }
   n35 = "Transform (Modify) Bone - Bone: dyn_hairB_R_03"!(ComponentPose=@n34, bAlphaBoolEnabled:bool=True, Translation="0,0,0", Rotation="0.008,-34.896,-0.043", Scale="1,1,1") -> (Pose) {
    alphaInputType=Bool
    rotationMode=BMM_Additive
    rotationSpace=BCS_BoneSpace
    scaleMode=BMM_Additive
    scaleSpace=BCS_BoneSpace
    translationMode=BMM_Additive
    translationSpace=BCS_BoneSpace
    alphaBoolBlend {blendInTime=0.2, blendOption=Linear, blendOutTime=0.3}
    alphaScaleBiasClamp {bInterpResult=true, interpSpeedIncreasing=20}
    boneToModify {boneName=dyn_hairB_R_03}
    rotation=<rot 0.008 -34.896 -0.043>
   }
   n36 = "Transform (Modify) Bone - Bone: base_hair_sd_g_r"!(ComponentPose=@n35, bAlphaBoolEnabled:bool=True, Translation="-0.211,-0.829,0.000", Rotation="-0.006,-23.386,0.002", Scale="1,1,1") -> (Pose) {
    alphaInputType=Bool
    rotationMode=BMM_Additive
    rotationSpace=BCS_BoneSpace
    scaleMode=BMM_Additive
    scaleSpace=BCS_BoneSpace
    translationMode=BMM_Additive
    translationSpace=BCS_BoneSpace
    alphaBoolBlend {blendInTime=0.2, blendOption=Linear, blendOutTime=0.3}
    alphaScaleBiasClamp {bInterpResult=true, interpSpeedIncreasing=20}
    boneToModify {boneName=base_hair_sd_g_r}
    rotation=<rot -0.006 -23.3859 0.002>
    translation=<-0.211 -0.829>
   }
   Save cached pose 'SetupNormalCache'(Pose=Component To Local(ComponentPose=@n36)) {cacheName=SetupNormalCache}
   n37 = RigidBody!(ComponentPose=Local To Component(LocalPose=Use cached pose 'ControlRig Cache'()), Alpha, OverrideWorldGravity=Get Gravity Override()) -> (Pose) {
    bClampLinearTranslationLimitToRefPose=true
    bOverrideWorldGravity=true
    overridePhysicsAsset=&/F_MED_LiftingRays_FaceAcc_PhysicsBackpack
    simulationSpace=BaseBoneSpace
    alphaBoolBlend {blendOption=Linear}
    baseBoneRef {boneName=head}
    componentAppliedLinearAccClamp=<3000 3000 3000>
    simSpaceSettings {dampingAlpha=0.1, maxAngularAcceleration=1, maxAngularVelocity=1, maxLinearAcceleration=1, maxLinearVelocity=1, velocityScaleZ=0.1}
   }
   Save cached pose 'WearingBackpackCache'(Pose=Component To Local(ComponentPose=@n37)) {cacheName=WearingBackpackCache}
  ]}
 EventGraph {nodes [
   Event OnCharacterCustomizationLoaded()
   > n1 = Cast To PlayerPawn_Athena(Object=TryGetPawnOwner())
   > Set Body_Mesh(Body_Mesh=GetSkeletalMeshForPartType(self=@n1))
   Event BlueprintUpdateAnimation()
   > n2 = Sequence!(execute) -> (then_0, then_1, then_2:Exec, then_3:Exec, then_4:Exec)
   > n3 = "Gravity Override State Function (Parameter Asset) (Nativized)"(
    CharacterPartAnimInstance=Self-Reference()
    GameTimeInSeconds=GetGameTimeInSeconds()
    DebugJointName=spine_05
    Parameters=&/M_MED_LiftingRays_FaceAcc_Parameters
   )
   > Set Gravity Override(Gravity Override=@n3)
   > Set RBAN Settings Out(RBAN Settings Out=@n3.OutRigidBodyAnimNodeParameters)
   n4 = NOT Boolean(A=Get bIsPlayingEmote())
   n5 = NOT Boolean(A=Get bIsSkyDiving())
   n6 = Get bIsFalling()
   Set FallingGround(execute=@n2.then_1, FallingGround:double="To Float (Boolean)"(InBool=AND Boolean!(A=@n6, B=@n5, C:bool=@n4) -> (ReturnValue))) -> (Output_Get:double)
   n7 = Get bIsMoving2D()
   n8 = Get bIsCrouching()
   > Set Moving2dCrouch(Moving2dCrouch:double="To Float (Boolean)"(InBool=AND Boolean!(A=@n7, B=@n8) -> (ReturnValue))) -> (Output_Get:double)
   n9 = NOT Boolean(A=Get bIsPlayingEmote())
   n10 = NOT Boolean(A=Get bIsOnGround() -> (bIsOnGround:bool))
   n11 = NOT Boolean(A=Get bIsSkyDiving())
   n12 = Get bIsFalling()
   > Set FallingSwimming(FallingSwimming:double="To Float (Boolean)"(InBool=AND Boolean!(A=@n12, B=@n11, C:bool=@n9, D:bool=@n10) -> (ReturnValue))) -> (Output_Get:double)
   n13 = NOT Boolean(A=Get bRigidBodyEnabled())
   n14 = Montage_IsPlaying(self=GetAnimInstance(self=GetOwnerSkeletalMeshComponent()), Montage=/BRCosmetics/Animation/Game/MainPlayer/Emotes/Emote_Celebration_Female_M)
   > Set PlayingCelebrationNoRban(PlayingCelebrationNoRban:double="To Float (Boolean)"(InBool=AND Boolean!(A=@n14, B=@n13) -> (ReturnValue))) -> (Output_Get:double)
   n15 = Get bRigidBodyEnabled()
   > Set PlayingCelebrationRban(PlayingCelebrationRban:double="To Float (Boolean)"(InBool=AND Boolean(A=@n14, B=@n15))) -> (Output_Get:double)
   n16 = NOT Boolean(A=Get bIsSkyDiving())
   n17 = NOT Boolean(A=Get bIsMoving2D())
   n18 = NOT Boolean(A=Get bIsPlayingEmote())
   n19 = Get bIsInFrontEnd()
   > Set Alpha_FEIdle(Alpha_FEIdle:double="To Float (Boolean)"(InBool=AND Boolean!(A=@n19, B=@n17, C:bool=@n18, D:bool=@n16) -> (ReturnValue))) -> (Output_Get:double)
   n20 = Sequence!(execute=@n2.then_3) -> (then_0, then_1)
   n21 = GetOwnerSkeletalMeshComponent()
   n22 = BreakVector(InVec=BreakTransform(InTransform=GetDeltaTransformFromRefPose(self=@n21, BoneName=pelvis, BaseName=root)))
   n23 = Get bIsSkydiveFloating()
   > n24 = Branch(Condition=AND Boolean!(A="float < float"(A=@n22.Z, B=-20), B=@n23) -> (ReturnValue))
   > Set Alpha_RightFloating(Alpha_RightFloating:double=1) -> (Output_Get:double)
   Set Alpha_RightFloating(execute=@n24.else, Alpha_RightFloating=0)
   n25 = Branch(execute=@n20.then_1, Condition=AND Boolean(A=@n23, B="float > float"(A=@n22.Z, B=20)))
   > Set Alpha_LeftFloating(Alpha_LeftFloating:double=1) -> (Output_Get:double)
   Set Alpha_LeftFloating(execute=@n25.else, Alpha_LeftFloating=0)
   n26 = Get bRigidBodyEnabled()
   n27 = CheckForBackbling()
   Set isWearingBackbling / Cape / NoRBAN(
    execute=@n2.then_2
    isWearingBackbling / Cape / NoRBAN:bool=OR Boolean!(A=@n27, B=@n27.bIsBackblingACape, C:bool=@n27.bIsBackblingLong, D:bool=NOT Boolean(A=@n26)) -> (ReturnValue)
   ) -> (Output_Get:bool)
   Event OnParachuteOpened()
   > Switch on EGliderType(Selection=Get PlayerGliderType() -> (PlayerGliderType=HangGlider))
   > Set HangGlidingAlpha(HangGlidingAlpha:double=1) -> (Output_Get:double)
   Event OnParachuteClosed()
   > Set HangGlidingAlpha(HangGlidingAlpha=0)
   Event BlueprintInitializeAnimation()
   > Set Skel Mesh(Skel Mesh=GetOwnerSkeletalMeshComponent())
   Event OnWeaponChanged()
   > n28 = Cast To FortPlayerAnimInstance(Object=GetAnimInstance(self=GetOwnerSkeletalMeshComponent()))
   > Set Pistol(Pistol:double=0) -> (Output_Get:double)
   > n29 = Switch on EFortWeaponCoreAnimation!(execute, Selection:EFortWeaponCoreAnimation Enum=Get WeaponCoreAnim(self=@n28)) -> (
    Melee:Exec
    Pistol:Exec
    Shotgun:Exec
    PaperBlueprint:Exec
    Rifle:Exec
    MeleeOneHand:Exec
    MachinePistol:Exec
    RocketLauncher:Exec
    GrenadeLauncher:Exec
    GoingCommando:Exec
    AssaultRifle:Exec
    TacticalShotgun:Exec
    SniperRifle:Exec
    TrapPlacement:Exec
    ShoulderLauncher:Exec
    AbilityDecoTool:Exec
    Crossbow:Exec
    C4:Exec
    RemoteControl:Exec
    DualWield:Exec
    AR_BullPup:Exec
    AR_ForwardGrip:Exec
    MedPackPaddles:Exec
    SMG_P90:Exec
    AR_DrumGun:Exec
    Consumable_Small:Exec
    Consumable_Large:Exec
    Balloon:Exec
    MountedTurret:Exec
    CreativeTool:Exec
    ExplosiveBow:Exec
    AshtonIndigo:Exec
    AshtonChicago:Exec
    MeleeDualWield:Exec
    Unarmed:Exec
   )
   Set Pistol(execute=@n29.Pistol, Pistol=1)
  ]}
]
variables [
 Alpha_FEIdle {type=number}
 Alpha_LeftFloating {type=number}
 Alpha_RightFloating {type=number}
 Body_Mesh {type=SkeletalMeshComponent}
 Component Linear Acc Scale {type=Vector}
 Component Linear Vel Scale {type=Vector}
 FallingGround {type=number}
 FallingSwimming {type=number}
 Gravity Override {type=Vector}
 Gravity Override Params {
  type=GravityOverrideParamsStruct
  default {
   riding_Moving_Targeting {gravityOverride=<0 0 -980>}
   skydive_FloatingDown {windAmplitude=<150 -450 300>}
   skydive_Parachuting {windAmplitude=<50 -100 100>, windFrequency=<1 4 5>}
  }
 }
 HangGlidingAlpha {type=number}
 isWearingBackbling / Cape / NoRBAN {type=boolean}
 Moving2dCrouch {type=number}
 Pistol {type=number}
 PlayingCelebrationNoRban {type=number}
 PlayingCelebrationRban {type=number}
 RBAN Params {
  type=RBANParamsStruct
  default {
   dBNO {simSpaceSettings {
     maxAngularAcceleration=50
     maxAngularVelocity=3
     maxLinearAcceleration=300
     maxLinearVelocity=700
     velocityScaleZ=0
     externalLinearDragV=<0 0.5 0>
    }}
   dBNO_Carried {simSpaceSettings {
     maxAngularAcceleration=50
     maxAngularVelocity=3
     maxLinearAcceleration=300
     maxLinearVelocity=700
     velocityScaleZ=0
     externalLinearDragV=<0 0.5 0>
    }}
   emoteOrMelee {simSpaceSettings {
     maxAngularAcceleration=1000
     maxAngularVelocity=3
     maxLinearAcceleration=300
     maxLinearVelocity=700
     velocityScaleZ=0
     externalLinearDragV=<0 0.5 0>
    }}
   front_End {simSpaceSettings {
     maxAngularAcceleration=1000
     maxAngularVelocity=3
     maxLinearAcceleration=300
     maxLinearVelocity=700
     velocityScaleZ=0
     externalLinearDragV=<0 0.5 0>
    }}
   onGround_Crouch_Moving {simSpaceSettings {
     maxAngularAcceleration=1000
     maxAngularVelocity=3
     maxLinearAcceleration=300
     maxLinearVelocity=700
     velocityScaleZ=0
     externalLinearDragV=<0 0.5 0>
    }}
   onGround_Falling {simSpaceSettings {
     maxAngularAcceleration=1000
     maxAngularVelocity=3
     maxLinearAcceleration=300
     maxLinearVelocity=300
     velocityScaleZ=1
     externalLinearDragV=<0 0 0>
    }}
   onGround_Moving {
    componentLinearAccScale=<0 0 0>
    componentLinearVelScale=<0 0 0>
    simSpaceSettings {
     maxAngularAcceleration=1000
     maxAngularVelocity=15
     maxLinearAcceleration=300
     maxLinearVelocity=300
     velocityScaleZ=1
     worldAlpha=0.25
     externalLinearDragV=<0 0 0>
    }
   }
   onGround_Standing {simSpaceSettings {
     maxAngularAcceleration=1000
     maxAngularVelocity=3
     maxLinearAcceleration=300
     maxLinearVelocity=700
     velocityScaleZ=0
     externalLinearDragV=<0 0.5 0>
    }}
   rail_Grinding {simSpaceSettings {
     maxAngularAcceleration=1000
     maxAngularVelocity=3
     maxLinearAcceleration=300
     maxLinearVelocity=700
     velocityScaleZ=0
     worldAlpha=0.1
     externalLinearDragV=<0 0.5 0>
    }}
   riding_Moving {simSpaceSettings {
     maxAngularAcceleration=1000
     maxAngularVelocity=3
     maxLinearAcceleration=300
     maxLinearVelocity=700
     velocityScaleZ=0
     externalLinearDragV=<0 0.5 0>
    }}
   riding_Moving_Targeting {simSpaceSettings {
     maxAngularAcceleration=1000
     maxAngularVelocity=3
     maxLinearAcceleration=300
     maxLinearVelocity=700
     velocityScaleZ=0
     externalLinearDragV=<0 0.5 0>
    }}
   riding_Standing {simSpaceSettings {
     maxAngularAcceleration=1000
     maxAngularVelocity=3
     maxLinearAcceleration=300
     maxLinearVelocity=700
     velocityScaleZ=0
     externalLinearDragV=<0 0.5 0>
    }}
   skydive_DivingDown {simSpaceSettings {
     maxAngularAcceleration=1000
     maxAngularVelocity=3
     maxLinearAcceleration=300
     maxLinearVelocity=700
     velocityScaleZ=0
     externalLinearDragV=<0 0.5 0>
    }}
   skydive_FloatingDown {
    componentLinearAccScale=<0 0 0>
    componentLinearVelScale=<0 0 0>
    simSpaceSettings {
     maxAngularAcceleration=1000
     maxAngularVelocity=15
     maxLinearAcceleration=300
     maxLinearVelocity=700
     velocityScaleZ=0.3
     worldAlpha=0.3
     externalLinearDragV=<0 8 0>
    }
   }
   skydive_GoingUp {simSpaceSettings {
     maxAngularAcceleration=1000
     maxAngularVelocity=3
     maxLinearAcceleration=300
     maxLinearVelocity=700
     velocityScaleZ=0
     externalLinearDragV=<0 0.5 0>
    }}
   skydive_Parachuting {
    componentLinearAccScale=<0 0.1 0.3>
    componentLinearVelScale=<0 0.1 0.3>
    simSpaceSettings {
     maxAngularAcceleration=1000
     maxAngularVelocity=3
     maxLinearAcceleration=300
     maxLinearVelocity=700
     velocityScaleZ=0
     externalLinearDragV=<0 0.5 0>
    }
   }
   skydive_Surfing {simSpaceSettings {
     maxAngularAcceleration=1000
     maxAngularVelocity=3
     maxLinearAcceleration=300
     maxLinearVelocity=700
     velocityScaleZ=0
     externalLinearDragV=<0 0.5 0>
    }}
   swimming {simSpaceSettings {
     maxAngularAcceleration=1000
     maxAngularVelocity=3
     maxLinearAcceleration=300
     maxLinearVelocity=700
     velocityScaleZ=0
     externalLinearDragV=<0 0.5 0>
    }}
   tactical_Sprint {simSpaceSettings {
     maxAngularAcceleration=1000
     maxAngularVelocity=3
     maxLinearAcceleration=300
     maxLinearVelocity=700
     velocityScaleZ=0
     externalLinearDragV=<0 0.5 0>
    }}
   ziplining {simSpaceSettings {
     maxAngularAcceleration=1000
     maxAngularVelocity=3
     maxLinearAcceleration=300
     maxLinearVelocity=700
     velocityScaleZ=0
     externalLinearDragV=<0 0.5 0>
    }}
  }
 }
 RBAN Settings Out {type=FortRigidBodyAnimNodeParameters}
 SimSpaceSettingsOld {
  type=SimSpaceSettings
  default {
   dampingAlpha=1
   maxAngularAcceleration=10000
   maxAngularVelocity=10000
   maxLinearAcceleration=10000
   maxLinearVelocity=10000
   velocityScaleZ=1
   worldAlpha=0
   externalLinearDragV=<0 0 0>
  }
 }
 Skel Mesh {type=SkeletalMeshComponent}
]
